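# One-shot stereo MetaSound for the Marksman Visor's enemy-spotted cue, assigned to the Fort_Gadgets sound class.
# On play, two wave players start together: a heartbeat and a HUD marked/targeted-by-opponent sound.
# A three-input stereo mixer sums both to the output, and the heartbeat's end signals that the source is finished.
# Character interface inputs and several wave asset inputs are declared but left unconnected.
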MetaSoundSource /MarksmanVisorGameplay/Assets/Audio/MSS_Marksman_Visor_Enemy_Spotted
inputs [
 Character.Affiliation {type=Enum:FortAudioAffiliation}
 Character.Health {type=Float}
 Character.UniqueID {type=Int32}
 In Array {type=WaveTableBankAsset:Array}
 In Array 1 {type=WaveAsset:Array}
 UE.Source.OnPlay {type=Trigger}
 Wave Asset {type=WaveAsset:Array}
 Wave Asset 1 {type=WaveAsset}
 Wave Asset 2 {type=WaveAsset}
]
interfaces [Fort.Character:1.0, UE.OutputFormat.Stereo:1.0, UE.Source.OneShot:1.0, UE.Source:1.1]
outputs [
 "UE.OutputFormat.Stereo.Audio:0" {type=Audio}
 "UE.OutputFormat.Stereo.Audio:1" {type=Audio}
 UE.Source.OneShot.OnFinished {type=Trigger}
]
pages [Default {
  input_defaults {
   Character.Affiliation=0
   Character.Health=0
   Character.UniqueID=0
   In Array 1=null
   UE.Source.OnPlay=null
   Wave Asset 1=null
   Wave Asset 2=null
   In Array [null]
   Wave Asset [null]
  }
  nodes [
   WavePlayer_0 = UE.Wave Player:Stereo(Play=^UE.Source.OnPlay) {also {defaults {Wave Asset=/MarksmanVisorGameplay/Assets/Audio/Waves/Marksman_Visor_Heart_Beat_02}}}
   WavePlayer_1 = UE.Wave Player:Stereo(Play=^UE.Source.OnPlay) {also {defaults {Wave Asset=/MarksmanVisorGameplay/Assets/Audio/Waves/HUD_Marked_Targeted_By_Opponent_01}}}
   AudioMixerAudioMixerStereo3_0 = "AudioMixer.Audio Mixer (Stereo, 3)"(In 2 L=@WavePlayer_0.Out Left, In 2 R=@WavePlayer_0.Out Right, In 1 L=@WavePlayer_1.Out Left, In 1 R=@WavePlayer_1.Out Right) {also {defaults {Gain 0=0}}}
   InitVariable.WaveAsset:Array()
  ]
  links [
   {dst="@UE.OutputFormat.Stereo.Audio:0", src=AudioMixerAudioMixerStereo3_0.Out L}
   {dst="@UE.OutputFormat.Stereo.Audio:1", src=AudioMixerAudioMixerStereo3_0.Out R}
   {dst="@UE.Source.OneShot.OnFinished", src=WavePlayer_0.On Finished}
  ]
 }]
properties {
 assetImportData=&self:AssetImportData
 numChannels=2
 outputFormat=Stereo
 sampleRate=48000
 soundClassObject=&Fort_Gadgets
}
