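# One-shot stereo MetaSound for the Marksman Visor's player-spotted cue, with spatial attenuation and the gadgets sound class.
# On play, a briefly delayed heartbeat always sounds, while an affiliation comparison picks a second layer.
# A true result plays the targeted-by-opponent sting immediately; false plays the opponent caret-placement sound after a longer delay.
# A three-input mixer favors the targeted sting, and only the heartbeat's completion signals OnFinished.
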MetaSoundSource /MarksmanVisorGameplay/Assets/Audio/MSS_Marksman_Visor_Player_Spotted
inputs [
 Character.Affiliation {type=Enum:FortAudioAffiliation}
 Character.Health {type=Float}
 Character.UniqueID {type=Int32}
 UE.Source.OnPlay {type=Trigger}
]
interfaces [Fort.Character:1.0, UE.OutputFormat.Stereo:1.0, UE.Source.OneShot:1.0, UE.Source:1.1]
outputs [
 "UE.OutputFormat.Stereo.Audio:0" {type=Audio}
 "UE.OutputFormat.Stereo.Audio:1" {type=Audio}
 UE.Source.OneShot.OnFinished {type=Trigger}
]
pages [Default {
  input_defaults {Character.Affiliation=0, Character.Health=0, Character.UniqueID=0, UE.Source.OnPlay=null}
  nodes [
   TriggerDelay_0 = UE.Trigger Delay(In=^UE.Source.OnPlay) {also {defaults {Delay Time=0.1}}}
   WavePlayer_0 = UE.Wave Player:Stereo(Play=@TriggerDelay_0.Out) {also {defaults {Wave Asset=/MarksmanVisorGameplay/Assets/Audio/Waves/Marksman_Visor_Heart_Beat_01}}}
   n1 = &TriggerCompareAffiliation(A=^Character.Affiliation, Compare=^UE.Source.OnPlay)
   WavePlayer_1 = UE.Wave Player:Stereo(Play=@n1.True) {also {defaults {Wave Asset=/MarksmanVisorGameplay/Assets/Audio/Waves/HUD_Marked_Targeted_By_Opponent_01}}}
   TriggerDelay_1 = UE.Trigger Delay(In=@n1.False) {also {defaults {Delay Time=0.3}}}
   WavePlayer_2 = UE.Wave Player:Stereo(Play=@TriggerDelay_1.Out) {also {defaults {Wave Asset=/MarksmanVisorGameplay/Assets/Audio/Waves/HUD_Marked_Opponent_Caret_Placement_01}}}
   AudioMixerAudioMixerStereo3_0 = "AudioMixer.Audio Mixer (Stereo, 3)"(
    In 2 L=@WavePlayer_0.Out Left
    In 2 R=@WavePlayer_0.Out Right
    In 1 L=@WavePlayer_1.Out Left
    In 1 R=@WavePlayer_1.Out Right
    In 0 L=@WavePlayer_2.Out Left
    In 0 R=@WavePlayer_2.Out Right
   ) {also {defaults {Gain 0=0.6, Gain 1=1.5, Gain 2=0.5}}}
   InitVariable.WaveAsset:Array()
  ]
  links [
   {dst="@UE.OutputFormat.Stereo.Audio:0", src=AudioMixerAudioMixerStereo3_0.Out L}
   {dst="@UE.OutputFormat.Stereo.Audio:1", src=AudioMixerAudioMixerStereo3_0.Out R}
   {dst="@UE.Source.OneShot.OnFinished", src=WavePlayer_0.On Finished}
  ]
 }]
properties {
 assetImportData=&self:AssetImportData
 attenuationSettings=&/ATT_Fort_MarksmanVisor_EnemySpoted
 maxDistance=32000
 numChannels=2
 outputFormat=Stereo
 sampleRate=48000
 soundClassObject=&Fort_Gadgets
}
