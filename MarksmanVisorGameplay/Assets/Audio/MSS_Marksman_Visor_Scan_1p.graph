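# One-shot stereo MetaSound source for the first-person Marksman Visor scan sound, with 2D-named attenuation, the gadgets sound class and a concurrency override.
# On play, it picks a random wave from an array and plays it alongside a separate extra-layer wave exposed as an input.
# Both players feed a stereo mixer, and a trigger accumulator reports finished once both layers end.
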MetaSoundSource /MarksmanVisorGameplay/Assets/Audio/MSS_Marksman_Visor_Scan_1p
inputs [
 Character.Affiliation {type=Enum:FortAudioAffiliation}
 Character.Health {type=Float}
 Character.UniqueID {type=Int32}
 ExtraLayer {type=WaveAsset}
 UE.Source.OnPlay {type=Trigger}
]
interfaces [Fort.Character:1.0, UE.OutputFormat.Stereo:1.0, UE.Source.OneShot:1.0, UE.Source:1.1]
outputs [
 "UE.OutputFormat.Stereo.Audio:0" {type=Audio}
 "UE.OutputFormat.Stereo.Audio:1" {type=Audio}
 UE.Source.OneShot.OnFinished {type=Trigger}
]
pages [Default {
  input_defaults {
   Character.Affiliation=0
   Character.Health=0
   Character.UniqueID=0
   ExtraLayer=/MarksmanVisorGameplay/Assets/Audio/Waves/Scan_extra_Layer
   UE.Source.OnPlay=null
  }
  nodes [
   ArrayRandomGet_0 = Array.Random Get:WaveAsset:Array(Next=^UE.Source.OnPlay, In Array=VariableAccessor.WaveAsset:Array(Variable=InitVariable.WaveAsset:Array().Variable).Value)
   WavePlayer_0 = UE.Wave Player:Stereo(Play=@ArrayRandomGet_0.On Next, Wave Asset=@ArrayRandomGet_0.Value)
   WavePlayer_1 = UE.Wave Player:Stereo(Wave Asset=^ExtraLayer, Play=^UE.Source.OnPlay)
   AudioMixerAudioMixerStereo2_0 = "AudioMixer.Audio Mixer (Stereo, 2)"(In 0 L=@WavePlayer_0.Out Left, In 0 R=@WavePlayer_0.Out Right, In 1 L=@WavePlayer_1.Out Left, In 1 R=@WavePlayer_1.Out Right)
   TriggerAccumulatorTriggerAccumulate2_0 = "TriggerAccumulator.Trigger Accumulate (2)"(In 0=@WavePlayer_0.On Finished, In 1=@WavePlayer_1.On Finished)
  ]
  links [
   {dst="@UE.OutputFormat.Stereo.Audio:0", src=AudioMixerAudioMixerStereo2_0.Out L}
   {dst="@UE.OutputFormat.Stereo.Audio:1", src=AudioMixerAudioMixerStereo2_0.Out R}
   {dst="@UE.Source.OneShot.OnFinished", src=TriggerAccumulatorTriggerAccumulate2_0.Out}
  ]
 }]
properties {
 assetImportData=&self:AssetImportData
 attenuationSettings=&/ATT_Marksman_Visor_Scan_2d
 bOverrideConcurrency=true
 numChannels=2
 outputFormat=Stereo
 sampleRate=48000
 soundClassObject=&Fort_Gadgets
 concurrencyOverrides {maxCount=3}
}
