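# One-shot stereo MetaSound for the third-person Marksman Visor scan, using an enemy-distant attenuation and limited concurrency.
# On play, it mixes a fixed extra layer with randomized pitch and a random pick from an array variable that shows no assigned waves.
# A wavetable curve over remapped listener distance sets the random layer's gain, with distance also logged.
# A trigger accumulator signals finish once both players end.
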MetaSoundSource /MarksmanVisorGameplay/Assets/Audio/MSS_Marksman_Visor_Scan_3p
inputs [
 ActiveSound.bIsOccluded {type=Bool}
 ActiveSound.MaxDistance {type=Float}
 Character.Affiliation {type=Enum:FortAudioAffiliation}
 Character.Health {type=Float}
 Character.UniqueID {type=Int32}
 ExtraLayer {type=WaveAsset}
 In Array {type=WaveTableBankAsset:Array}
 TEST {type=Float}
 UE.Attenuation.Distance {type=Float}
 UE.Source.OnPlay {type=Trigger}
 Wave Asset {type=WaveAsset:Array}
 Wave Asset 2 {type=WaveAsset}
]
interfaces [Fort.ActiveSound:1.0, Fort.Character:1.0, UE.Attenuation:1.0, UE.OutputFormat.Stereo:1.0, UE.Source.OneShot:1.0, UE.Source:1.1]
outputs [
 "UE.OutputFormat.Stereo.Audio:0" {type=Audio}
 "UE.OutputFormat.Stereo.Audio:1" {type=Audio}
 UE.Source.OneShot.OnFinished {type=Trigger}
]
pages [Default {
  input_defaults {
   ActiveSound.MaxDistance=0
   ActiveSound.bIsOccluded=null
   Character.Affiliation=0
   Character.Health=0
   Character.UniqueID=0
   ExtraLayer=/MarksmanVisorGameplay/Assets/Audio/Waves/Scan_extra_Layer
   TEST=0
   UE.Attenuation.Distance=0
   UE.Source.OnPlay=null
   Wave Asset 2=null
   In Array [null]
   Wave Asset [null]
  }
  nodes [
   n1 = UE.RandomFloat(Next=^UE.Source.OnPlay) {also {defaults {Min=0.95}}}
   ArrayRandomGet_0 = Array.Random Get:WaveAsset:Array(Next=@n1.On Next, In Array=VariableAccessor.WaveAsset:Array(Variable=InitVariable.WaveAsset:Array().Variable).Value)
   WavePlayer_0 = UE.Wave Player:Stereo(Wave Asset=^ExtraLayer, Play=@ArrayRandomGet_0.On Next, Pitch Shift=@n1.Value)
   WavePlayer_1 = UE.Wave Player:Stereo(Wave Asset=@ArrayRandomGet_0.Value, Play=UE.Reroute(Value=^UE.Source.OnPlay).Value)
   AudioMixerAudioMixerStereo2_0 = "AudioMixer.Audio Mixer (Stereo, 2)"(
    In 0 L=@WavePlayer_0.Out Left
    In 0 R=@WavePlayer_0.Out Right
    In 1 L=@WavePlayer_1.Out Left
    In 1 R=@WavePlayer_1.Out Right
    Gain 1=UE.WaveTableBankEvaluate(Input=MapRange.MapRange:Float(In=^UE.Attenuation.Distance) {also {defaults {In Range A=8000, In Range B=1000, Out Range A=1, Out Range B=0}}}.Out Value) {also {defaults {WaveTableBank=&/WTB_VisonAtten}}}.Output
   ) {also {defaults {Gain 0=0.36}}}
   PrintLogPrintLog_0 = Print Log.Print Log:Float(Value To Log=^UE.Attenuation.Distance, Trigger=@WavePlayer_1.On Play) {also {defaults {Label="***Distance: "}}}
   TriggerAccumulatorTriggerAccumulate2_0 = "TriggerAccumulator.Trigger Accumulate (2)"(In 0=@WavePlayer_0.On Finished, In 1=@WavePlayer_1.On Finished)
  ]
  links [
   {dst="@UE.OutputFormat.Stereo.Audio:0", src=AudioMixerAudioMixerStereo2_0.Out L}
   {dst="@UE.OutputFormat.Stereo.Audio:1", src=AudioMixerAudioMixerStereo2_0.Out R}
   {dst="@UE.Source.OneShot.OnFinished", src=TriggerAccumulatorTriggerAccumulate2_0.Out}
  ]
 }]
properties {
 assetImportData=&self:AssetImportData
 attenuationSettings=&/ATT_Marksman_Visor_Scan_Enemy_Distant
 bOverrideConcurrency=true
 maxDistance=21000
 numChannels=2
 outputFormat=Stereo
 sampleRate=48000
 soundClassObject=&Fort_Gadgets
 concurrencyOverrides {maxCount=3}
}
