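# One-shot stereo MetaSound source for the Marksman Visor zoom-in sound.
# On play, it rolls a random float, then picks a random wave from an array and plays it with that float as pitch shift.
# The wave player drives both outputs and signals finished when the clip ends.
# It uses the gadgets sound class, the visor's pickup attenuation, low volume and a concurrency override that caps simultaneous instances.
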MetaSoundSource /MarksmanVisorGameplay/Assets/Audio/MSS_Marksman_Visor_Zoom_In
inputs [UE.Source.OnPlay {type=Trigger}]
interfaces [UE.OutputFormat.Stereo:1.0, UE.Source.OneShot:1.0, UE.Source:1.1]
outputs [
 "UE.OutputFormat.Stereo.Audio:0" {type=Audio}
 "UE.OutputFormat.Stereo.Audio:1" {type=Audio}
 UE.Source.OneShot.OnFinished {type=Trigger}
]
pages [Default {
  input_defaults {UE.Source.OnPlay=null}
  nodes [
   n1 = UE.RandomFloat(Next=^UE.Source.OnPlay) {also {defaults {Max=2, Min=1}}}
   ArrayRandomGet_0 = Array.Random Get:WaveAsset:Array(Next=@n1.On Next, In Array=VariableAccessor.WaveAsset:Array(Variable=InitVariable.WaveAsset:Array().Variable).Value)
   WavePlayer_0 = UE.Wave Player:Stereo(Play=@ArrayRandomGet_0.On Next, Wave Asset=@ArrayRandomGet_0.Value, Pitch Shift=@n1.Value)
  ]
  links [
   {dst="@UE.Source.OneShot.OnFinished", src=WavePlayer_0.On Finished}
   {dst="@UE.OutputFormat.Stereo.Audio:0", src=WavePlayer_0.Out Left}
   {dst="@UE.OutputFormat.Stereo.Audio:1", src=WavePlayer_0.Out Right}
  ]
 }]
properties {
 assetImportData=&self:AssetImportData
 attenuationSettings=&/ATT_Fort_MarksmanVisor_PickUp
 bOverrideConcurrency=true
 maxDistance=2300
 numChannels=2
 outputFormat=Stereo
 sampleRate=48000
 soundClassObject=&Fort_Gadgets
 volume=0.1
 concurrencyOverrides {maxCount=2}
}
