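# Translucent unlit VFX material for god rays on particles and meshes, with a glitchy, shimmering look.
# Emissive combines particle color, a sphere-masked glow, optional curve-atlas gradient and the green channel of an animated glitch flipbook.
# Opacity blends distorted glitch textures with noise panned by a dynamic parameter, shaped by UV edge fades, optional Fresnel and particle alpha.
# Particle random desynchronizes the animation; depth fade is skipped on mobile.
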Material /MarksmanVisorGameplay/Assets/FX/M_GodRays
graph {
 properties {
  bUsedWithInstancedStaticMeshes=true
  bUsedWithNanite=true
  bUsedWithNiagaraMeshParticles=true
  bUsedWithNiagaraSprites=true
  bUsedWithSplineMeshes=true
  blendMode=BLEND_Translucent
  pixelDepthOffsetMode=PDOM_Legacy
  shadingModel=MSM_Unlit
  usedShadingModels=Unlit
 }
 nodes [
  n1 = ComponentMask(None=NamedRerouteUsage()) {g=true}
  n2 = ParticleColor()
  n3 = SphereMask(
   A=TextureCoordinate()
   B=MaterialFunctionCall(X=Constant() {r=0.5}, Y=Constant() {r=0.7}) {materialFunction=&MakeFloat2}.Result
   Radius=Constant() {r=0.5}
   Hardness=Constant()
  )
  n4 = LinearInterpolate(
   A=LinearInterpolate(A=@n2.RGB, B=VectorParameter() {parameterName=Glow, defaultValue=<rgba 15 9.7031 0.45 0>}.RGB, Alpha=Power(Base=@n3) {constExponent=3})
   B=@n2.RGB
   Alpha=ScalarParameter() {parameterName=JustParticleColor}
  )
  n5 = TextureCoordinate()
  n6 = CurveAtlasRowParameter(CurveTime=OneMinus(None=ComponentMask(None=@n5) {r=true})) {
   atlas=/Game/Athena/LinearColorCurves/VFX_CurveAtlas
   curve=/Game/Accessories/FORT_Backpacks/Backpack_F_MED_Quarrel/FX/Curve_Quarrel_Auroras_02
   defaultValue=29
   group=Curve
   parameterName=Curve
  }
  n7 = Multiply(
   A=LinearInterpolate(A=Constant3Vector() {constant=<rgba 1 1 1 0>}, B=@n6, Alpha=ScalarParameter() {defaultValue=1, parameterName=Color Curve Intensity})
   B=@n4
  )
  n8 = ScalarParameter() {parameterName=Use Color Curve}
  MaterialFunctionCall(Emissive=Add(A=Multiply(A=@n1, B=@n2.RGB), B=LinearInterpolate(A=@n4, B=@n7, Alpha=@n8))) {materialFunction=&MF_FX_EmissiveAdjustments}
  n9 = ComponentMask(None=TextureCoordinate() {vTiling=0}) {r=true}
  n10 = ScalarParameter() {defaultValue=1, parameterName=PatternTileU}
  n11 = ComponentMask(None=Panner(Coordinate=Multiply(A=@n9, B=@n10), Time=ScalarParameter() {parameterName=Pattern_PanU}) {speedX=0.5}) {r=true}
  n12 = ComponentMask(None=TextureCoordinate() {uTiling=0}) {g=true}
  n13 = ScalarParameter() {defaultValue=1, parameterName=PatternTileV}
  n14 = ComponentMask(None=Panner(Coordinate=Multiply(A=@n12, B=@n13), Time=ScalarParameter() {parameterName=Pattern_PanV}) {speedY=0.5}) {g=true}
  n15 = Constant() {r=1}
  n16 = ScalarParameter() {defaultValue=8, parameterName=Columns}
  n17 = Divide(A=@n15, B=@n16)
  n18 = ScalarParameter() {defaultValue=8, parameterName=Rows}
  n19 = Divide(A=@n15, B=@n18)
  n20 = Multiply(A=Add(A=Time(), B=ParticleRandom()), B=ScalarParameter() {defaultValue=16, parameterName=FB_Speed})
  n21 = Fmod(A=Floor(None=@n20), B=@n18)
  n22 = Floor(None=Divide(A=@n20, B=@n18))
  n23 = Add(
   A=Multiply(A=AppendVector(A=@n11, B=@n14), B=AppendVector(A=@n17, B=@n19))
   B=AppendVector(A=Divide(A=@n21, B=@n18), B=Divide(A=@n22, B=@n16))
  )
  n24 = Time()
  n25 = ParticleRandom()
  n26 = Multiply(A=ScalarParameter() {defaultValue=0.01, parameterName=GlitchSpeed}, B=Sine(None=Add(A=@n24, B=@n25)))
  n27 = TextureSample(UVs=Panner(Coordinate=TextureCoordinate() {uTiling=0.25, vTiling=0.25}, Time=@n26) {speedX=-1, speedY=0.1}) {texture=/Game/Effects/Fort_Effects/Textures/Glitches/T_VCVFX_Boost_Glitch_Pixel_Alpha}
  n28 = TextureSample(UVs=Panner(Coordinate=TextureCoordinate() {vTiling=-0.1}, Time=@n26) {speedX=-0.5, speedY=0.1}) {texture=&T_VCVFX_Boost_Glitch_Pixel_Alpha}
  n29 = Add(A=Time(), B=ParticleRandom())
  n30 = LinearInterpolate(
   A=@n23
   B=Add(A=@n23, B=AppendVector(A=@n27.R, B=@n28.G))
   Alpha=Multiply(A=ScalarParameter() {defaultValue=0.15, parameterName=GlitchIntensity}, B=Sine(None=@n29) {period=8})
  )
  NamedRerouteDeclaration(None=TextureSample(UVs=Panner(Coordinate=@n30) {speedX=-1.125}) {texture=/Game/Effects/Fort_Effects/Textures/Glitches/T_VCVFX_Boost_Glitch_Glitch_FB_03_RGB}.RGB) {name=GLITCH_RGB, nodeColor=<rgba 0 0.2266 1 1>}
  n31 = Multiply(A=NamedRerouteUsage()) {constB=2.5}
  n32 = TextureCoordinate()
  n33 = Multiply(A=ScalarParameter() {defaultValue=2, parameterName=MaskGlitchSpeed}, B=Add(A=Time(), B=ParticleRandom()))
  n34 = TextureSample(UVs=Panner(Coordinate=TextureCoordinate() {vTiling=0.25}, Time=@n33) {speedX=-0.25, speedY=0.1}) {texture=&T_VCVFX_Boost_Glitch_Pixel_Alpha}
  n35 = TextureSample(UVs=Panner(Coordinate=TextureCoordinate() {uTiling=2, vTiling=-0.1}, Time=@n33) {speedX=-0.1, speedY=0.1}) {texture=&T_VCVFX_Boost_Glitch_Pixel_Alpha}
  n36 = ComponentMask(None=TextureSampleParameter2D(UVs=AppendVector(A=@n34.R, B=@n35.G)) {parameterName=GlitchTexture, texture=&T_VCVFX_Boost_Glitch_Glitch_FB_03_RGB}.RGB) {g=true, r=true}
  n37 = ScalarParameter() {defaultValue=0.1, parameterName=MaskGlitchIntensity}
  n38 = Sine(None=Add(A=Time(), B=ParticleRandom())) {period=8}
  n39 = Panner(Coordinate=LinearInterpolate(A=Add(A=@n32, B=@n36), B=@n32, Alpha=Multiply(A=@n37, B=@n38)), Time=Constant() {r=0.25}) {speedY=-0.1}
  n40 = ConstantBiasScale(None=Multiply(A=TextureSample(UVs=@n39) {texture=/Game/Effects/Fort_Effects/Textures/Glitches/T_VCVFX_Boost_Glitch_RGB_Mask_RGB}.R, B=Constant() {r=1})) {bias=-0.01, scale=1}
  n41 = Saturate(None=Multiply(A=Saturate(None=@n40), B=Constant() {r=0.5}))
  n42 = ScalarParameter() {defaultValue=1.5, parameterName=Texture_Power}
  n43 = ScalarParameter() {defaultValue=1, parameterName=TextureTile}
  n44 = TextureCoordinate() {vTiling=0}
  n45 = DynamicParameter() {paramNames [PanningSpeed,Param2,Param3,Param4]}
  n46 = TextureSampleParameter2D(UVs=Panner(
    Coordinate=Multiply(A=@n43, B=@n44)
    Speed=MaterialFunctionCall(X=@n45.PanningSpeed) {materialFunction=&MakeFloat2}.Result
   )) {parameterName=Tex, texture=&T_NoisePack_01}
  n47 = LinearInterpolate(
   A=Multiply(A=Add(A=@n31, B=@n41)) {constB=7.5}
   B=Multiply(A=Power(Base=@n46.R, Exp=@n42) {constExponent=1.35}, B=ScalarParameter() {defaultValue=0.5, parameterName=Texture_Multiply}) {constB=3}
  ) {constAlpha=0.35}
  n48 = TextureCoordinate()
  n49 = Multiply(A=Multiply(A=@n48, B=OneMinus(None=@n48))) {constB=4}
  n50 = Multiply(A=Multiply(A=ComponentMask(None=@n49) {r=true}, B=ComponentMask(None=@n49) {g=true})) {constB=2.5}
  n51 = ComponentMask(None=@n48) {g=true}
  n52 = LinearInterpolate(A=@n51, B=Multiply(A=@n51, B=OneMinus(None=@n51)), Alpha=ScalarParameter() {parameterName=FadeBottom})
  n53 = ScalarParameter() {defaultValue=2, parameterName=FadeTop}
  n54 = Multiply(A=Multiply(A=Saturate(None=@n50), B=Power(Base=@n52, Exp=@n53))) {constB=0.5}
  n55 = MaterialFunctionCall(Exponent=Constant() {r=3}) {materialFunction=&MF_Fresnel_FX}
  n56 = ScalarParameter() {defaultValue=1, parameterName=UseFresnel}
  n57 = Multiply(A=@n2.A, B=Multiply(A=Multiply(A=@n47, B=@n54), B=LinearInterpolate(B=@n55.Fresnel, Alpha=@n56) {constA=1}))
  n58 = DepthFade(Opacity=@n57, FadeDistance=ScalarParameter() {parameterName=DepthFade_Distance})
  ShadingPathSwitch(Default=@n58, Deferred=@n58, Mobile=@n57) {desc=Migrated from FeatureLevelSwitch}
 ]
}
