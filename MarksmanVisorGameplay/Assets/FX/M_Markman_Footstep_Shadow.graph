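# Unlit translucent Niagara sprite and mesh particle material that, by its name, draws the footstep shadow for the Marksman visor effects.
# A soft SphereMask circle centred in UV space, multiplied by particle alpha, forms the fading spot shape.
# A black-by-default Basic Color parameter feeds the shared MF_FX_EmissiveAdjustments function for colour.
# The CameraOffset world position function, driven by a scalar parameter, shifts the particle along the view direction.
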Material /MarksmanVisorGameplay/Assets/FX/M_Markman_Footstep_Shadow
graph {
 properties {
  bAutomaticallySetUsageInEditor=false
  bUsedWithNiagaraMeshParticles=true
  bUsedWithNiagaraSprites=true
  blendMode=BLEND_Translucent
  shadingModel=MSM_Unlit
  usedShadingModels=Unlit
 }
 nodes [
  MaterialFunctionCall(Offset Amount=ScalarParameter() {defaultValue=200, parameterName=CameraOffset}) {materialFunction=/Engine/Functions/Engine_MaterialFunctions02/WorldPositionOffset/CameraOffset}
  MaterialFunctionCall(Emissive=VectorParameter() {parameterName=Basic Color, defaultValue=<rgba 0 0 0 1>}.RGB) {materialFunction=&MF_FX_EmissiveAdjustments}
  Multiply(A=SphereMask(A=TextureCoordinate(), B=Constant() {r=0.5}) {attenuationRadius=0.25, hardnessPercent=50}, B=ParticleColor().A)
 ]
}
