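# Unlit, additive, two-sided material for the Marksman Visor selection box, drawing green with depth testing disabled.
# A Progress parameter raises a bright line up local Z, normalized by object bounds, with a faint fill below it.
# An inverted UV box mask adds borders that pulse on a time-driven sine while Active is set.
# Active also blends from dim uniform green to brighter green with teal highlights.
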Material /MarksmanVisorGameplay/Assets/FX/M_MarksmanVisor_Selection_Box
graph {
 properties {
  bAutomaticallySetUsageInEditor=false
  bCastRayTracedShadows=false
  bDisableDepthTest=true
  blendMode=BLEND_Additive
  shadingModel=MSM_Unlit
  twoSided=true
  usedShadingModels=Unlit
 }
 nodes [
  n1 = ScalarParameter() {defaultValue=1, parameterName=Active}
  n2 = Multiply(A=Constant3Vector() {constant=<rgba 0 1 0 1>}, B=LinearInterpolate(Alpha=@n1) {constA=0.25})
  n3 = Divide(A=LocalPosition().XYZ, B=ObjectLocalBounds().Half Extents)
  n4 = NamedRerouteUsage()
  n5 = Divide(A=Add(A=ComponentMask(None=@n3) {b=true}, B=LinearInterpolate(Alpha=@n4) {constA=1.025, constAlpha=0, constB=-1})) {constB=0.025}
  n6 = Saturate(None=@n5)
  n7 = Abs(None=@n5)
  n8 = LinearInterpolate(A=Multiply(A=OneMinus(None=@n6)) {constB=0.33}, Alpha=OneMinus(None=Saturate(None=@n7)))
  n9 = LinearInterpolate(
   B=MaterialFunctionCall!(Sine Phase=Time(), Value 1=Constant() {r=0.5}, Value 2=Constant() {r=1}) {materialFunction=/Engine/Functions/Engine_MaterialFunctions02/Utility/Sine_Remapped}.Remapped Sine
   Alpha=ScalarParameter() {defaultValue=1, parameterName=Active}
  ) {constA=1}
  n10 = OneMinus(None=MaterialFunctionCall!(
    A=TextureCoordinate()
    B=Constant() {r=0.5}
    Bounds=ScalarParameter() {defaultValue=0.9, parameterName=2d Bounds}
    Edge Falloff=ScalarParameter() {defaultValue=0.05}
   ) {materialFunction=/Engine/Functions/Engine_MaterialFunctions01/Gradient/BoxMask-2D}.Result)
  LinearInterpolate(
   A=@n2
   B=LinearInterpolate(A=@n2, B=Constant3Vector() {constant=<rgba 0 2 0.75 1>}, Alpha=@n1)
   Alpha=Power(Base=LinearInterpolate(A=@n8, B=@n9, Alpha=@n10)) {constExponent=1.5}
  )
  MaterialFunctionCall() {materialFunction=&ObjectScale}
  Modulo!(A=Multiply(A=Time()) {constB=0.25}, B=Constant() {r=1})
  NamedRerouteDeclaration(None=ScalarParameter() {parameterName=Progress}) {name=Progress, nodeColor=<rgba 1 0.3984 0 1>}
 ]
}
