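# Additive, unlit Niagara sprite material for glitchy footstep markers in Marksman Visor gameplay, with camera offset.
# An animated flipbook, desynchronized per particle and warped by sine-driven glitch pixel textures, supplies the GLITCH_RGB pattern.
# That pattern appears to distort a soft sphere mask, which blends dark and light colors and is scaled by particle alpha.
# Dynamic parameter Param2 shifts both colors toward targets before particle color and MF_FX_EmissiveAdjustments shape the emissive.
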Material /MarksmanVisorGameplay/Assets/FX/M_Marksman_Footsteps
graph {
 properties {
  bAutomaticallySetUsageInEditor=false
  bCastRayTracedShadows=false
  bUsedWithNiagaraSprites=true
  blendMode=BLEND_Additive
  shadingModel=MSM_Unlit
  twoSided=true
  usedShadingModels=Unlit
 }
 nodes [
  MaterialFunctionCall(Offset Amount=ScalarParameter() {defaultValue=200, parameterName=CameraOffset}) {materialFunction=/Engine/Functions/Engine_MaterialFunctions02/WorldPositionOffset/CameraOffset}
  n1 = DynamicParameter() {defaultValue=<rgba 1 0 1 1>}
  n2 = LinearInterpolate(
   A=VectorParameter() {parameterName=Base Color Dark, defaultValue=<rgba 0.2 0.2 0.2 1>}.RGB
   B=VectorParameter() {parameterName=Target Color Dark, defaultValue=<rgba 1 1 1 1>}.RGB
   Alpha=@n1.Param2
  )
  n3 = LinearInterpolate(
   A=VectorParameter() {parameterName=Base Color Light, defaultValue=<rgba 1 1 1 1>}.RGB
   B=VectorParameter() {parameterName=Target Color LIght, defaultValue=<rgba 2 2 2 1>}.RGB
   Alpha=@n1.Param2
  )
  n4 = ParticleColor()
  n5 = TextureCoordinate()
  n6 = Multiply(A=ComponentMask(None=NamedRerouteUsage()) {g=true, r=true}, B=Constant() {r=0.2}) {constB=0.2}
  n7 = Multiply(A=@n4.A, B=SphereMask(A=Add(A=@n5, B=@n6), B=Constant() {r=0.5}) {attenuationRadius=0.25, hardnessPercent=0})
  n8 = Saturate(None=Power(Base=@n7))
  n9 = Constant() {r=5}
  MaterialFunctionCall(Emissive=Multiply(A=LinearInterpolate(A=@n2, B=@n3, Alpha=@n8), B=Multiply(A=@n4.RGB, B=@n9) {constB=5})) {materialFunction=&MF_FX_EmissiveAdjustments}
  n10 = NamedRerouteUsage()
  Multiply(A=LinearInterpolate(A=@n7, B=Multiply(A=@n7, B=@n10)) {constAlpha=0.75}) {constB=2.5}
  n11 = ComponentMask(None=TextureCoordinate() {vTiling=0}) {r=true}
  n12 = ScalarParameter() {defaultValue=0.5, parameterName=PatternTileU}
  n13 = ComponentMask(None=Panner(Coordinate=Multiply(A=@n11, B=@n12), Time=ScalarParameter() {parameterName=Pattern_PanU}) {speedX=0.5}) {r=true}
  n14 = ComponentMask(None=TextureCoordinate() {uTiling=0}) {g=true}
  n15 = ScalarParameter() {defaultValue=0.5, parameterName=PatternTileV}
  n16 = ComponentMask(None=Panner(Coordinate=Multiply(A=@n14, B=@n15), Time=ScalarParameter() {parameterName=Pattern_PanV}) {speedY=0.5}) {g=true}
  n17 = Constant() {r=1}
  n18 = ScalarParameter() {defaultValue=8, parameterName=Columns}
  n19 = Divide(A=@n17, B=@n18)
  n20 = ScalarParameter() {defaultValue=8, parameterName=Rows}
  n21 = Divide(A=@n17, B=@n20)
  n22 = Multiply(A=Add(A=Time(), B=ParticleRandom()), B=ScalarParameter() {defaultValue=16, parameterName=FB_Speed})
  n23 = Fmod(A=Floor(None=@n22), B=@n20)
  n24 = Floor(None=Divide(A=@n22, B=@n20))
  n25 = Add(
   A=Multiply(A=AppendVector(A=@n13, B=@n16), B=AppendVector(A=@n19, B=@n21))
   B=AppendVector(A=Divide(A=@n23, B=@n20), B=Divide(A=@n24, B=@n18))
  )
  n26 = Time()
  n27 = ParticleRandom()
  n28 = Multiply(A=ScalarParameter() {defaultValue=0.01, parameterName=GlitchSpeed}, B=Sine(None=Add(A=@n26, B=@n27)))
  n29 = TextureSample(UVs=Panner(Coordinate=TextureCoordinate() {uTiling=0.1, vTiling=0.1}, Time=@n28) {speedX=-1, speedY=0.1}) {texture=/Game/Effects/Fort_Effects/Textures/Glitches/T_VCVFX_Boost_Glitch_Pixel_Alpha}
  n30 = TextureSample(UVs=Panner(Coordinate=TextureCoordinate() {uTiling=0.1, vTiling=-0.1}, Time=@n28) {speedX=-0.5, speedY=0.1}) {texture=&T_VCVFX_Boost_Glitch_Pixel_Alpha}
  n31 = Add(A=Time(), B=ParticleRandom())
  n32 = LinearInterpolate(
   A=@n25
   B=Add(A=@n25, B=AppendVector(A=@n29.R, B=@n30.G))
   Alpha=Multiply(A=ScalarParameter() {defaultValue=2, parameterName=GlitchIntensity}, B=Sine(None=@n31) {period=8})
  )
  NamedRerouteDeclaration(None=TextureSample(UVs=Panner(Coordinate=@n32) {speedX=-1.125}) {texture=/Game/Effects/Fort_Effects/Textures/Glitches/T_VCVFX_Boost_Glitch_Glitch_FB_03_RGB}.RGB) {name=GLITCH_RGB, nodeColor=<rgba 0 0.2266 1 1>}
 ]
}
