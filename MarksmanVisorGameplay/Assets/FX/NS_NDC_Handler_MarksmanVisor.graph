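# Niagara Data Channel handler for the Marksman Visor that draws glowing footprint trails from channel data, with user inputs for instigator affiliation and lifetime.
# One Foot_Step emitter spawns mirrored prints aligned to movement, rendered through four sprite renderers and a mesh renderer.
# A travelling waveform drives brightness and ray height, while color shifts from green along a green-yellow-red gradient.
# The tracked target's prints change material parameters and color timing; a matching kill ID or timeline end removes particles.
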NiagaraSystem /MarksmanVisorGameplay/Assets/FX/NS_NDC_Handler_MarksmanVisor
dependencies {
 usedDataInterfaces [
  &NiagaraDataInterfaceArrayFloat
  &NiagaraDataInterfaceArrayFloat2
  &NiagaraDataInterfaceCurve
  /Script/Niagara.NiagaraDataInterfaceDataChannelRead
 ]
 usedDynamicInputs [
  &SelectFloatFromArray
  &InvertBool
  /Game/Effects/Niagara/DynamicInputs/State/FN_IsExecutionStateActive
 ]
 usedModules [
  &SystemState
  /Niagara/Modules/System/CompleteIfUnused
  &self:SystemSpawnScript.SystemScriptSource.SystemScriptGraph.NiagaraNodeAssignment_0.SetVariables_#1
  &self:SystemSpawnScript.SystemScriptSource.SystemScriptGraph.NiagaraNodeAssignment_1.SetVariables_#2
 ]
}
emitters {Foot_Step {
  emitterUpdateScript [
   &EmitterState()
   &/NDC_MarksmanVisor_Spawn(InstigatorAffiliation:NiagaraInt32=$User.InstigatorAffiliation, Reader:NiagaraDataInterfaceDataChannelRead=$System.NDCReader)
  ]
  particleSpawnScript [
   &/NDC_MarksmanVisor_InitializeParticle(SequenceNumber=60)
   /SetVariables(
    Particles.Direction:Vector3f=&NormalizeVector(Vector To Normalize=$Particles.ProjectileVelocity, Safe Normalize=true, Safe Fallback Threshold=0, Safe Fallback Vector=<0 0 0.001>)
    Particles.FootSize:NiagaraFloat=$Particles.FootSize
    Particles.Normal:Vector3f=$Particles.OwnerVelocity
   ) {isSetParameters=true}
   &InitializeParticle(
    Lifetime=$User.FootstepDuration
    Position=$Particles.CurrentPosition
    Position Mode=::NewEnumerator1
    Sprite Size=&Multiply_Vector2D_ByFloat(Float=$Particles.FootSize, Vector2D=<0.5 1>)
    Sprite Size Mode=::NewEnumerator4
    Sprite UV Mode=::NewEnumerator4
    Sprite UV Scale:Vector2f=&SelectVector2DFromArray(Direct Array Index=$Particles.PositionType, Vector 2D Selection Array=DataInterface() {propertyValues {FloatData [<1 1>, <-1 1>]}})
    UsePositionOffset=false
    Write Material Random=true
   )
   &SpriteFacingAndAlignment(Sprite Alignment=$Particles.Direction, Sprite Facing=$Particles.Normal, WriteAlignment=true, WriteFacing=true)
   /SetVariables(Particles.SpriteAlignmentGlow:Vector3f=<0 0 1>) {isSetParameters=true}
   /SetVariables() {isSetParameters=true}
   /SetVariables(Particles.Color_Rays:LinearColor=<rgba 0.0289 1 0.0017 1>) {isSetParameters=true}
  ]
  particleUpdateScript [
   &ParticleState()
   /SetVariables(Particles.CameraOffsetRays=10) {isSetParameters=true}
   /SetVariables(Particles.IsTargeted:NiagaraBool=&SetBoolByIntComparison(A=$User.TrackedTarget1, B=$Particles.InstigatorInt, Comparison Type=::NewEnumerator4)) {isSetParameters=true}
   &Timeline(Play=$Particles.IsTargeted, Play Rate=2, Timeline End=0.5)
   /SetVariables(Particles.NormalizedAgeNew:NiagaraFloat=&Lerp_Float(
     A=&ClampFloat(Float=&Divide_Float(A=$Particles.NormalizedAge, B=0.5))
     B=$Particles.NormalizedAge
     Alpha=$Particles.Timeline.NormalizedTime
    )) {isSetParameters=true}
   /SetVariables(Particles.TrailMarquee:NiagaraFloat=&Power(
     a=&Waveform(Global Phase=$Particles.Age, "Waveform [1]"=::NewEnumerator0, "[1] Frequency"=0.5, "[1] Amplitude Min/Max"=<0.33 1>, "[1] Exponent"=2)
     b=1
    )) {isSetParameters=true}
   &DynamicMaterialParameters(
    Index 0 Param 1=&MakeCustomFloatFromBool(Bool=$Particles.IsTargeted, True Float=0.6, False Float=0.2)
    Index 0 Param 2=&MakeCustomFloatFromBool(Bool=$Particles.IsTargeted, True Float=1, False Float=0)
    Module.Index0Float0Write=true
    Module.Index0Float1Write=true
    Module.Index0Float2Write=true
    Module.Index0Float3Write=true
   )
   /SetVariables(Particles.CustomWave:NiagaraFloat=&Lerp_Float(
     A=1
     B=&Waveform(
      Global Phase=&Subtract_Float(
       A=&Multiply_Float(A=$Engine.Time, B=$User.CC_ScaleTime)
       B=&Multiply_FloatByInt(
        Float=&Divide_Float(A=1, B=&ClampFloat(Float=$User.CC_Frequency, Min=5, Max=20))
        Integer=&ReturnParticleIDAttributes()
       )
      )
      "Waveform [1]"=::NewEnumerator0
      "[1] Frequency"=1
      "[1] Amplitude Min/Max"=<0 1>
      "[1] Exponent"=$User.CC_Exponent
     )
     Alpha=&FloatFromCurve(
      FloatCurve=DataInterface() {propertyValues {Curve {
         defaultValue=1
         keys [
          {arriveTangent=0.2601, interpMode=RCIM_Cubic, leaveTangent=0.2601, tangentMode=RCTM_User, time=0, value=0}
          {arriveTangent=1.7223, interpMode=RCIM_Cubic, leaveTangent=1.7223, tangentMode=RCTM_User, time=0.6109, value=0.2979}
          {arriveTangent=0.0224, interpMode=RCIM_Cubic, leaveTangent=0.0224, tangentMode=RCTM_User, time=1, value=1}
         ]
        }}}
      CurveIndex=$Particles.Age
     )
    )) {enabled=false, isSetParameters=true}
   &self:WaveForm(
    Exponent:NiagaraFloat=$User.CC_Exponent
    Frequency:NiagaraFloat=$User.CC_Frequency
    LengthScale:NiagaraFloat=$User.CC_LengthScale
    StepsPerTrail:NiagaraInt32=$User.CC_StepsPerTrail
    TimeMult:NiagaraFloat=$User.CC_ScaleTime
   ) {module=WaveForm002}
   /SetVariables(Particles.CustomWave:NiagaraFloat=&Lerp_Float(
     A=$User.CC_LeadingEdgeHeight
     B=$Particles.Waveform
     Alpha=&FloatFromCurve(
      FloatCurve=DataInterface() {propertyValues {Curve {
         defaultValue=1
         keys [
          {arriveTangent=0.2601, interpMode=RCIM_Cubic, leaveTangent=0.2601, tangentMode=RCTM_User, time=0, value=0}
          {arriveTangent=1.7223, interpMode=RCIM_Cubic, leaveTangent=1.7223, tangentMode=RCTM_User, time=0.6109, value=0.2979}
          {arriveTangent=0.0224, interpMode=RCIM_Cubic, leaveTangent=0.0224, tangentMode=RCTM_User, time=1, value=1}
         ]
        }}}
      CurveIndex=&Divide_Float(A=$Particles.Age, B=$User.CC_LeadingEdgeSeconds)
     )
    )) {isSetParameters=true}
   &ScaleColor(
    Scale Alpha=&Multiply_Float(
     A=$Particles.CustomWave
     B=&FloatFromCurve(FloatCurve=DataInterface() {propertyValues {Curve {
         defaultValue=1
         keys [
          {arriveTangent=157.1932, interpMode=RCIM_Cubic, leaveTangent=157.1932, tangentMode=RCTM_User, time=0, value=0}
          {interpMode=RCIM_Cubic, tangentMode=RCTM_User, time=0.0213, value=1.0068}
          {arriveTangent=-3.5669, interpMode=RCIM_Cubic, leaveTangent=-3.5669, tangentMode=RCTM_User, time=1, value=0}
         ]
        }}})
    )
    Scale RGB=&MakeVectorFromLinearColorRGB(Color=&LerpLinearColors(
      StartColor=<rgba 0.0289 1 0.0017 1>
      EndColor=&self:Lerp_3Color(
       Linear Color:LinearColor=<rgba 0.0289 1 0.0017 1>
       Linear Color001:LinearColor=<rgba 1 0.9382 0.0068 1>
       Linear Color002:LinearColor=<rgba 1 0.0294 0 1>
       float:NiagaraFloat=&OneMinusFloat(Float=&ReturnNormalizedExecIndex())
      )
      LerpFactor=&FloatFromCurve(
       FloatCurve=DataInterface() {propertyValues {Curve {
          defaultValue=1
          keys [
           {interpMode=RCIM_Cubic, tangentMode=RCTM_User, time=0, value=0}
           {arriveTangent=2.8157, interpMode=RCIM_Cubic, leaveTangent=2.8157, tangentMode=RCTM_User, time=1, value=1}
          ]
         }}}
       CurveIndex=$Particles.Age
      )
     ))
    ScaleA=true
    ScaleRGB=true
   ) {enabled=false}
   &ScaleColor(
    Scale Alpha=&Multiply_Float(
     A=$Particles.CustomWave
     B=&FloatFromCurve(FloatCurve=DataInterface() {propertyValues {Curve {
         defaultValue=1
         keys [
          {arriveTangent=157.1932, interpMode=RCIM_Cubic, leaveTangent=157.1932, tangentMode=RCTM_User, time=0, value=0}
          {interpMode=RCIM_Cubic, tangentMode=RCTM_User, time=0.0213, value=1.0068}
          {arriveTangent=-3.5669, interpMode=RCIM_Cubic, leaveTangent=-3.5669, tangentMode=RCTM_User, time=1, value=0}
         ]
        }}})
    )
    Scale RGB=&MakeVectorFromLinearColorRGB(Color=&LerpLinearColors(
      StartColor=<rgba 0.0289 1 0.0017 1>
      EndColor=&self:Lerp_3Color(
       Linear Color:LinearColor=<rgba 0.0289 1 0.0017 1>
       Linear Color001:LinearColor=<rgba 1 0.9382 0.0068 1>
       Linear Color002:LinearColor=<rgba 1 0.0294 0 1>
       float:NiagaraFloat=&OneMinusFloat(Float=&ReturnNormalizedExecIndex())
      )
      LerpFactor=&FloatFromCurve(
       FloatCurve=DataInterface() {propertyValues {Curve {
          defaultValue=1
          keys [
           {interpMode=RCIM_Cubic, tangentMode=RCTM_User, time=0, value=0}
           {arriveTangent=2.8157, interpMode=RCIM_Cubic, leaveTangent=2.8157, tangentMode=RCTM_User, time=1, value=1}
          ]
         }}}
       CurveIndex=$Particles.NormalizedAgeNew
      )
     ))
    ScaleA=true
    ScaleRGB=true
   ) {module=ScaleColor004}
   &ScaleColor(
    Scale Alpha=&Multiply_Float(
     A=$Particles.CustomWave
     B=&FloatFromCurve(FloatCurve=DataInterface() {propertyValues {Curve {
         defaultValue=1
         keys [
          {arriveTangent=157.1932, interpMode=RCIM_Cubic, leaveTangent=157.1932, tangentMode=RCTM_User, time=0, value=0}
          {interpMode=RCIM_Cubic, tangentMode=RCTM_User, time=0.0213, value=1.0068}
          {arriveTangent=-3.5669, interpMode=RCIM_Cubic, leaveTangent=-3.5669, tangentMode=RCTM_User, time=1, value=0}
         ]
        }}})
    )
    Scale RGB=&MakeVectorFromLinearColorRGB(Color=<rgba 0.0289 1 0.0017 1>)
    ScaleA=true
    ScaleRGB=true
   ) {enabled=false, module=ScaleColor005}
   &ScaleColor(
    Scale Alpha=&Multiply_Float(
     A=&MakeFloatFromLinearColor(LinearColor=$Particles.Color, Channel:ENiagaraFloatFromLinearColorOptions=::NewEnumerator3)
     B=&Lerp_Float(
      A=&Waveform(
       "Waveform [1]"=::NewEnumerator10
       "[1] Frequency"=&RandomRangeFloat(Minimum=20, Maximum=40, Evaluation Type=::NewEnumerator1)
       "[1] Amplitude Min/Max"=<1 1.25>
      )
      B=2
      Alpha=$Particles.CustomWave
     )
    )
    ScaleA=true
    ScaleRGB=false
   ) {module=ScaleColor003}
   &ScaleColor(
    Scale Alpha=&FloatFromCurve(
     FloatCurve=DataInterface() {propertyValues {Curve {
        defaultValue=1
        keys [
         {tangentMode=RCTM_User, time=0, value=1}
         {arriveTangent=-0.0405, interpMode=RCIM_Cubic, leaveTangent=-0.0405, tangentMode=RCTM_User, time=0.7009, value=1}
         {arriveTangent=-8.0089, interpMode=RCIM_Cubic, leaveTangent=-8.0089, tangentMode=RCTM_User, time=1, value=0}
        ]
       }}}
     CurveIndex=$System.Timeline.NormalizedTime
    )
    ScaleA=true
    ScaleRGB=false
   ) {module=ScaleColor002}
   &ScaleColor(
    Scale Alpha=&Multiply_Float(
     A=&MakeFloatFromLinearColor(LinearColor=$Particles.Color, Channel:ENiagaraFloatFromLinearColorOptions=::NewEnumerator3)
     B=&FloatFromCurve(FloatCurve=DataInterface() {propertyValues {Curve {keys [<at 0 1>,<at 0.8 0.5>,<at 1 0>]}}})
    )
    ScaleA=true
    ScaleRGB=false
   ) {module=ScaleColor001}
   /SetVariables(Particles.SpriteSize_ShadowFootStep:Vector2f=&Multiply_Vector2D_ByFloat(Float=1.1, Vector2D=$Particles.SpriteSize)) {isSetParameters=true}
   /SetVariables(Particles.SpriteSize_Lightray:Vector2f=&MakeVector2D(X=25, Y=&Lerp_Float(A=0, B=120, Alpha=$Particles.CustomWave))) {isSetParameters=true}
   /SetVariables(Particles.SpriteSize_GlitchBits:Vector2f=&Multiply_Vector2D(A=<30 30>, B=<1 1>)) {isSetParameters=true}
   &KillParticles(Kill Particles=&SetBoolByFloatComparison(A=$System.Timeline.NormalizedTime, B=1, Comparison Type=::NewEnumerator2))
   &KillParticles(Kill Particles=&SetBoolByIntComparison(A=$Particles.Initial.InstigatorInt, B=$User.KillUniqueID, Comparison Type=::NewEnumerator4)) {module=KillParticles001}
  ]
  rendererClasses [&NiagaraSpriteRendererProperties, &NiagaraMeshRendererProperties]
  renderers [
   {rendererClass=&NiagaraSpriteRendererProperties}
   {rendererClass=&NiagaraSpriteRendererProperties, rendererIndex=1}
   {rendererClass=&NiagaraSpriteRendererProperties, rendererIndex=2}
   {rendererClass=&NiagaraSpriteRendererProperties, rendererIndex=3}
   {rendererClass=&NiagaraMeshRendererProperties, rendererIndex=4}
  ]
 }}
summary {
 systemName=NS_NDC_Handler_MarksmanVisor
 emitters [{emitterName=Foot_Step, rendererClasses [&NiagaraSpriteRendererProperties, &NiagaraMeshRendererProperties]}]
 userVariables [
  User.CC_Exponent {type=NiagaraFloat, defaultValue=10, description=""}
  User.CC_Frequency {type=NiagaraFloat, defaultValue=1, description=""}
  User.CC_LeadingEdgeHeight {type=NiagaraFloat, defaultValue=1, description=""}
  User.CC_LeadingEdgeSeconds {type=NiagaraFloat, defaultValue=1, description=""}
  User.CC_LengthScale {type=NiagaraFloat, defaultValue=4, description=""}
  User.CC_ScaleTime {type=NiagaraFloat, defaultValue=0.25, description=""}
  User.CC_StepsPerTrail {type=NiagaraInt32, defaultValue=60}
  User.FootstepDuration {type=NiagaraFloat, defaultValue=30, description=""}
  User.InstigatorAffiliation {type=NiagaraInt32}
  User.KillUniqueID {type=NiagaraInt32, defaultValue=-1}
  User.TrackedTarget1 {type=NiagaraInt32}
 ]
}
