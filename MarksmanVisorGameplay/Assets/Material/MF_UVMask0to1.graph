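# Material function that outputs a binary mask from a UV input, marking where the coordinates lie within the 0 to 1 square, for the MarksmanVisorGameplay content.
# It tests U and V separately with If nodes against thresholds inset slightly from 0 and 1.
# The four results multiply together, giving a hard-edged 1 inside and 0 outside, including a thin border at each edge.
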MaterialFunction /MarksmanVisorGameplay/Assets/Material/MF_UVMask0to1
graph {nodes [
  n1 = FunctionInput() {inputName=uv, inputType=FunctionInput_Vector2}
  n2 = ComponentMask(None=@n1) {r=true}
  n3 = Constant() {r=1}
  n4 = If(A=@n2, B=Constant() {r=0.99}, "A > B"=Constant(), "A == B"=@n3, "A < B"=@n3)
  n5 = Constant() {r=1}
  n6 = If(A=@n2, B=Constant() {r=0.01}, "A > B"=@n5, "A == B"=@n5, "A < B"=Constant())
  n7 = ComponentMask(None=@n1) {g=true}
  n8 = Constant() {r=1}
  n9 = If(A=@n7, B=Constant() {r=0.99}, "A > B"=Constant(), "A == B"=@n8, "A < B"=@n8)
  n10 = Constant() {r=1}
  n11 = If(A=@n7, B=Constant() {r=0.01}, "A > B"=@n10, "A == B"=@n10, "A < B"=Constant())
  FunctionOutput(None=Multiply(A=Multiply(A=@n4, B=@n6), B=Multiply(A=@n9, B=@n11)))
 ]}
