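# Translucent, two-sided skeletal-mesh material for the Marksman Visor's animated screen HUD, with idle, deploy, activation and targeting looks.
# Packed mask textures, placed by custom UV remaps, form the animated aimer, rotating circle, bar, frame and flow layers.
# Timed activation and targeting parameters drive color blends, plus a targeting mask and ring.
# Vertex colors group pieces for world-position-offset motion, including idle bobbing, state offsets and collapsing toward part centers through local-space math.
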Material /MarksmanVisorGameplay/Assets/Material/M_ScreenVisor_V02
graph {
 properties {
  bAutomaticallySetUsageInEditor=false
  bUsedWithMeshDeformer=true
  bUsedWithSkeletalMesh=true
  blendMode=BLEND_Translucent
  translucencyLightingMode=TLM_VolumetricDirectional
  twoSided=true
  usedShadingModels=Default Lit
 }
 nodes [
  n1 = Frac(None=Multiply(A=Time(), B=ScalarParameter() {defaultValue=0.2, group=ScreenAnimation, parameterName=IdlingSpeed}) {constB=3})
  n2 = Multiply(A=Constant2Vector() {g=1}, B=Step(Y=@n1) {constX=0.2, constY=0.1})
  Add(A=@n2)
  Add(A=NamedRerouteUsage(), B=NamedRerouteUsage())
  Add(A=NamedRerouteUsage(), B=NamedRerouteUsage())
  n3 = WorldPosition()
  MaterialExpressionTransformPosition_1 = TransformPosition(None=@n3.XYZ) {transformSourceType=TRANSFORMPOSSOURCE_World}
  n4 = Convert(None=@MaterialExpressionTransformPosition_1) {
   nodeName=Break Float3
   convertInputs [{
     type=Vector3
     expressionInput {
      expression=&self:MaterialExpressionTransformPosition_1
      inputName=None
      mask=0
      maskA=0
      maskB=0
      maskG=0
      maskR=0
      outputIndex=0
     }
    }]
   convertOutputs [{type=Vector3}]
  }
  n5 = VectorParameter() {group=TechArt_Reactivity, parameterName=UpperpartCenter01, defaultValue=<rgba 0 0 0 1>}
  n6 = NamedRerouteUsage()
  n7 = Multiply(A=@n5.RGB, B=Constant3Vector() {constant=<rgba -1 1 1 0>})
  n8 = Add(A=Multiply(A=Subtract(A=@n5.RGB, B=@n6), B=NamedRerouteUsage()), B=Multiply(A=Subtract(A=@n7, B=@n6), B=NamedRerouteUsage()))
  n9 = VectorParameter() {group=TechArt_Reactivity, parameterName=DownpartCenter01, defaultValue=<rgba 0 0 0 1>}
  n10 = NamedRerouteUsage()
  n11 = Multiply(A=Subtract(A=@n9.RGB, B=@n10), B=NamedRerouteUsage())
  n12 = Constant3Vector() {constant=<rgba -1 1 1 0>}
  n13 = Subtract(A=Multiply(A=@n9.RGB, B=@n12), B=@n10)
  n14 = NamedRerouteUsage()
  n15 = VectorParameter() {group=TechArt_Reactivity, parameterName=MiddlepartCenter01, defaultValue=<rgba 0 0 0 1>}
  n16 = NamedRerouteUsage()
  n17 = Add(A=Add(A=Add(A=@n8, B=@n11), B=Multiply(A=@n13, B=@n14)), B=Multiply(A=Subtract(A=@n15.RGB, B=@n16), B=NamedRerouteUsage()))
  n18 = Multiply(A=Subtract(A=Multiply(A=@n15.RGB, B=@n12), B=@n16), B=NamedRerouteUsage())
  MaterialExpressionAdd_42 = Add(
   A=NamedRerouteDeclaration(None=@n4) {name=LocalPos, nodeColor=<rgba 0 0.2734 1 1>}
   B=NamedRerouteDeclaration(None=Add(A=@n17, B=@n18)) {name=WPOinLocal, nodeColor=<rgba 1 0.75 0 1>}
  )
  n19 = Convert(None=@MaterialExpressionAdd_42) {
   nodeName=Break Float3
   convertInputs [{
     type=Vector3
     expressionInput {
      expression=&self:MaterialExpressionAdd_42
      inputName=None
      mask=0
      maskA=0
      maskB=0
      maskG=0
      maskR=0
      outputIndex=0
     }
    }]
   convertOutputs [{type=Vector3}]
  }
  n20 = Constant3Vector() {constant=<rgba 0 3 7 0>}
  Add(A=@n19, B=@n20)
  Constant3Vector() {constant=<rgba 0 0.1302 0.0467 1>}
  Constant() {r=1}
  Constant() {r=1}
  Constant()
  Constant()
  Constant() {r=1}
  n21 = VectorParameter() {parameterName=VisorColor, defaultValue=<rgba 0 1 0 1>}
  n22 = Multiply(A=@n21.RGB, B=@n21.RGB)
  n23 = VectorParameter() {group=ScreenAnimation, parameterName=TargetingRingColor, defaultValue=<rgba 0.316 1 0 5>}
  n24 = VectorParameter() {group=ScreenAnimation, parameterName=TargetingRingColorActivate, defaultValue=<rgba 0.316 1 0 1.5>}
  n25 = NamedRerouteUsage()
  n26 = LinearInterpolate(A=Multiply(A=@n23.RGB, B=@n23.RGB), B=Multiply(A=@n24.RGB, B=@n24.RGB), Alpha=@n25)
  n27 = NamedRerouteUsage()
  n28 = Sign(None=@n25)
  n29 = NamedRerouteUsage()
  n30 = VectorParameter() {group=ScreenAnimation, parameterName=AimerColor01, defaultValue=<rgba 0.2451 1 0.0034 1>}
  n31 = Multiply(A=@n30.RGB, B=@n30.RGB)
  n32 = VectorParameter() {group=ScreenAnimation, parameterName=AimerColorActivate01, defaultValue=<rgba 1 0 0.0095 1>}
  n33 = Multiply(A=@n32.RGB, B=@n32.RGB)
  n34 = NamedRerouteUsage()
  n35 = VectorParameter() {group=ScreenAnimation, parameterName=RotateColor01, defaultValue=<rgba 0.0764 0.2281 0.008 1>}
  n36 = Multiply(A=@n35.RGB, B=@n35.RGB)
  n37 = VectorParameter() {group=ScreenAnimation, parameterName=RotateColor_Activate, defaultValue=<rgba 1 1 1 1>}
  n38 = Multiply(A=@n37.RGB, B=@n37.RGB)
  n39 = NamedRerouteUsage()
  n40 = Add(
   A=Multiply(A=LinearInterpolate(A=@n31, B=@n33, Alpha=@n34), B=NamedRerouteUsage())
   B=Multiply(A=NamedRerouteUsage(), B=LinearInterpolate(A=@n36, B=@n38, Alpha=@n39))
  )
  n41 = ScalarParameter() {defaultValue=2, group=ScreenAnimation, parameterName=Overall_EmissiveIntensity}
  n42 = ScalarParameter() {defaultValue=4, group=ScreenAnimation, parameterName=Overall_EmissiveIntensity_Targeting}
  n43 = NamedRerouteUsage()
  n44 = Add(
   A=Multiply(A=Multiply(A=@n26, B=@n27), B=Saturate(None=@n28))
   B=Multiply(A=Add(A=@n29, B=@n40), B=LinearInterpolate(A=@n41, B=@n42, Alpha=@n43))
  )
  Divide(
   A=StaticSwitchParameter(True=VertexColor(), False=StaticSwitchParameter(True=@n22, False=@n44) {group=Vis, parameterName="[Debug]Pickup Mesh"}) {group=Vis, parameterName="[Debug]VtxColor"}
   B=EyeAdaptation().EyeAdaptation
  )
  n45 = NamedRerouteUsage()
  n46 = Step(Y=@n45) {constX=0.35}
  n47 = Step(Y=@n45) {constX=0.45}
  Multiply(A=OneMinus(None=@n46), B=@n47)
  n48 = Step(Y=@n45) {constX=0.55}
  n49 = Step(Y=@n45) {constX=0.65}
  Multiply(A=OneMinus(None=@n48), B=@n49)
  n50 = VectorParameter() {group=ScreenAnimation, parameterName=ScreenFresenelColor01, defaultValue=<rgba 0.0764 0.2281 0.008 1>}
  n51 = VectorParameter() {group=ScreenAnimation, parameterName=ScreenFresenelColorActivate, defaultValue=<rgba 0.2281 0.0081 0 1>}
  n52 = NamedRerouteUsage()
  n53 = Fresnel() {baseReflectFraction=0, exponent=4}
  n54 = ScalarParameter() {defaultValue=2, group=ScreenAnimation, parameterName=FresnelIntenisty}
  n55 = Add(
   A=Multiply(A=LinearInterpolate(A=@n50.RGB, B=@n51.RGB, Alpha=@n52), B=Multiply(A=@n53, B=@n54))
   B=LinearInterpolate(
    A=VectorParameter() {group=ScreenAnimation, parameterName=ScreenColor01, defaultValue=<rgba 0.0094 0.0281 0.001 1>}.RGB
    B=VectorParameter() {group=ScreenAnimation, parameterName=ScreenColorActivate01, defaultValue=<rgba 0 0 0 1>}.RGB
    Alpha=NamedRerouteUsage()
   )
  )
  n56 = VectorParameter() {group=ScreenAnimation, parameterName=BarColor01, defaultValue=<rgba 0.3349 1 0.0351 1>}
  n57 = VectorParameter() {group=ScreenAnimation, parameterName=BarColorActivate01, defaultValue=<rgba 1 0.1114 0 1>}
  n58 = LinearInterpolate(A=Multiply(A=@n56.RGB, B=@n56.RGB), B=Multiply(A=@n57.RGB, B=@n57.RGB), Alpha=NamedRerouteUsage())
  n59 = NamedRerouteUsage()
  n60 = VectorParameter() {group=ScreenAnimation, parameterName=FocusFrameColor01, defaultValue=<rgba 1 1 1 1>}
  n61 = Multiply(A=@n60.RGB, B=@n60.RGB)
  n62 = VectorParameter() {group=ScreenAnimation, parameterName=FocusFrameColorActivate, defaultValue=<rgba 0.5323 0 0.0002 1>}
  n63 = Multiply(A=@n62.RGB, B=@n62.RGB)
  n64 = NamedRerouteUsage()
  n65 = VectorParameter() {group=ScreenAnimation, parameterName=BarIconColor01, defaultValue=<rgba 1 1 1 1>}
  n66 = VectorParameter() {group=ScreenAnimation, parameterName=BarIconActivate, defaultValue=<rgba 0.5323 0 0.0002 1>}
  n67 = LinearInterpolate(
   A=LinearInterpolate(A=LinearInterpolate(A=@n55, B=@n58, Alpha=@n59), B=LinearInterpolate(A=@n61, B=@n63, Alpha=@n64), Alpha=NamedRerouteUsage())
   B=LinearInterpolate(A=Multiply(A=@n65.RGB, B=@n65.RGB), B=Multiply(A=@n66.RGB, B=@n66.RGB), Alpha=NamedRerouteUsage())
   Alpha=NamedRerouteUsage()
  )
  n68 = Multiply(A=TextureCoordinate()) {constB=3}
  Multiply(A=@n67, B=LinearInterpolate(Alpha=TextureSample(UVs=@n68) {texture=/MarksmanVisorGameplay/Assets/Texture/T_VisorDynamic}.R) {constA=0.5})
  n69 = TextureObjectParameter() {group=TechArt_Screen, parameterName=Bar, samplerType=SAMPLERTYPE_Masks, texture=/MarksmanVisorGameplay/Assets/Texture/T_Visor_Mask02_02}
  NamedRerouteDeclaration(None=@n69) {name=Bar_Texture, nodeColor=<rgba 0 1 0.3906 1>}
  n70 = VertexColor()
  n71 = Step(Y=@n70.R) {constX=0.4}
  n72 = Step(Y=@n70.R) {constX=0.8}
  n73 = Multiply(A=OneMinus(None=@n71), B=@n72)
  n74 = OneMinus(None=@n72)
  n75 = Add(A=Add(A=Multiply(A=@n71) {constB=0.33}, B=Multiply(A=@n73) {constB=0.66}), B=Multiply(A=@n74) {constB=0.99})
  n76 = Add(A=Time(), B=ScalarParameter() {defaultValue=0.5, group=ScreenAnimation, parameterName=GradientSpeed})
  NamedRerouteDeclaration(None=Frac(None=Add(A=@n75, B=@n76))) {name=Offset_Vtx, nodeColor=<rgba 0.3359 1 0 1>}
  n77 = TextureObjectParameter() {group=TechArt_Screen, parameterName=Aimer, samplerType=SAMPLERTYPE_Masks, texture=/MarksmanVisorGameplay/Assets/Texture/T_Visor_Mask01_01}
  NamedRerouteDeclaration(None=@n77) {name=AimerTexture, nodeColor=<rgba 0.1953 0 1 1>}
  n78 = VectorParameter() {group=UV, parameterName=Aimer_CounterUV01, channelNames {a=ScaleY, b=ScaleX, g=CenterY, r=CenterX}, defaultValue=<rgba 0.5 0.5 1 1>}
  MaterialExpressionAppendVector_2 = AppendVector(A=@n78.CenterX, B=@n78.CenterX)
  MaterialExpressionAppendVector_3 = AppendVector(A=@n78.ScaleX, B=@n78.ScaleX)
  MaterialExpressionTextureCoordinate_0 = TextureCoordinate()
  n79 = Custom!(UVPoint=@MaterialExpressionAppendVector_2, UVScale=@MaterialExpressionAppendVector_3, OriginalUV=@MaterialExpressionTextureCoordinate_0) {
   code="float2 DistUV =OriginalUV-UVPoint;\r\nDistUV = DistUV/UVScale + float2(0.5,0.5);\r\n\r\nreturn DistUV;"
   description=CopyToPointUV
   outputType=CMOT_Float2
   inputs [
    {inputName=UVPoint, input {expression=&self:MaterialExpressionAppendVector_2}}
    {inputName=UVScale, input {expression=&self:MaterialExpressionAppendVector_3}}
    {inputName=OriginalUV, input {expression=&self:MaterialExpressionTextureCoordinate_0}}
   ]
  }
  n80 = ScalarParameter() {defaultValue=0.03, group=TechArt_Screen, parameterName=Idle_Dist}
  n81 = Step(Y=@n1) {constX=0.3, constY=0.2}
  n82 = Multiply(A=Constant2Vector() {r=1}, B=Multiply(A=OneMinus(None=@n81), B=Step(Y=@n1) {constX=0.5, constY=0.3}))
  n83 = Constant2Vector() {r=-1}
  n84 = Multiply(A=OneMinus(None=Step(Y=@n1) {constX=0.6, constY=0.4}), B=Step(Y=@n1) {constX=0.8, constY=0.5})
  n85 = OneMinus(None=Step(Y=@n1) {constX=0.9, constY=0.6})
  n86 = Step(Y=@n1) {constX=0.99, constY=0.7}
  n87 = Add(A=Add(A=Add(A=@n2, B=@n82), B=Multiply(A=@n83, B=@n84)), B=Multiply(A=Constant2Vector() {g=-1}, B=Multiply(A=@n85, B=@n86)))
  NamedRerouteDeclaration(None=Add(A=@n79, B=Multiply(A=@n80, B=@n87))) {name=AimerUV, nodeColor=<rgba 1 0.3516 0 1>}
  n88 = TextureCoordinate()
  n89 = Add(
   A=Multiply(A=ComponentMask(None=@n88) {r=true}) {constB=1.5}
   B=Multiply(A=Time(), B=ScalarParameter() {defaultValue=0.5, group=ScreenAnimation, parameterName=FLowingSpeed})
  )
  n90 = LinearInterpolate(A=ScalarParameter() {defaultValue=0.3, group=ScreenAnimation, parameterName=MinFlowing}, Alpha=ConstantBiasScale(None=Sine(None=@n89)))
  n91 = VectorParameter() {group=ScreenAnimation, parameterName=FlowingColor01, defaultValue=<rgba 0 1 0.2193 1>}
  n92 = VectorParameter() {group=ScreenAnimation, parameterName=FlowingActivate01, defaultValue=<rgba 0.5323 0 0.0002 1>}
  n93 = LinearInterpolate(A=Multiply(A=@n91.RGB, B=@n91.RGB), B=Multiply(A=@n92.RGB, B=@n92.RGB), Alpha=NamedRerouteUsage())
  n94 = Multiply(A=Multiply(A=Multiply(A=@n90, B=@n93), B=NamedRerouteUsage()), B=OneMinus(None=NamedRerouteUsage()))
  n95 = VectorParameter() {group=ScreenAnimation, parameterName=OutlineColor01, defaultValue=<rgba 0 1 0.2193 1>}
  n96 = VectorParameter() {group=ScreenAnimation, parameterName=OutlineActivate01, defaultValue=<rgba 0.5323 0 0.0002 1>}
  n97 = LinearInterpolate(
   A=@n67
   B=LinearInterpolate(A=Multiply(A=@n95.RGB, B=@n95.RGB), B=Multiply(A=@n96.RGB, B=@n96.RGB), Alpha=NamedRerouteUsage())
   Alpha=Add(A=NamedRerouteUsage(), B=NamedRerouteUsage())
  )
  n98 = ScalarParameter() {defaultValue=4, group=TechArt_Reactivity, parameterName=OverallColor}
  n99 = ScalarParameter() {defaultValue=1, group=TechArt_Reactivity, parameterName=OverallColor_Deploy}
  n100 = ScalarParameter() {defaultValue=1, group=TechArt_Reactivity, parameterName=Depoly_Alpha}
  NamedRerouteDeclaration(None=Multiply(A=Add(A=@n94, B=@n97), B=LinearInterpolate(A=@n98, B=@n99, Alpha=@n100))) {name=ScreenColor, nodeColor=<rgba 0 1 0.3438 1>}
  n101 = Rotator(Coordinate=NamedRerouteUsage()) {speed=0.6}
  n102 = NamedRerouteUsage()
  NamedRerouteDeclaration(None=Multiply(
    A=MaterialFunctionCall!(uv=@n101) {materialFunction=&/MF_UVMask0to1}.Result
    B=TextureSample(UVs=@n101, Tex=@n102).G
   )) {name=RotateCircle, nodeColor=<rgba 1 0.0234 0 1>}
  n103 = NamedRerouteUsage()
  n104 = NamedRerouteUsage()
  NamedRerouteDeclaration(None=Multiply(
    A=MaterialFunctionCall(uv=@n103) {materialFunction=&MF_UVMask0to1}.Result
    B=TextureSample(UVs=@n103, Tex=@n104) {samplerSource=SSM_Wrap_WorldGroupSettings}.R
   )) {name=AimerUI, nodeColor=<rgba 0 0.2969 1 1>}
  n105 = OneMinus(None=NamedRerouteUsage())
  n106 = LinearInterpolate(B=NamedRerouteUsage(), Alpha=Saturate(None=Sign(None=@n105))) {constA=1}
  n107 = NamedRerouteUsage()
  n108 = Power(Base=NamedRerouteUsage())
  n109 = NamedRerouteUsage()
  n110 = ScalarParameter() {defaultValue=0.1, group=TechArt_Screen, parameterName=Mask_Opacity01}
  n111 = Add(A=NamedRerouteUsage(), B=Add(A=NamedRerouteUsage(), B=Add(A=@n109, B=@n110)))
  n112 = LinearInterpolate(A=ScalarParameter() {defaultValue=0.3, group=ScreenAnimation, parameterName=MinOpacityGradientU}, Alpha=NamedRerouteUsage()) {constA=0.2}
  n113 = ScalarParameter() {defaultValue=0.5, parameterName=SideOpacity}
  n114 = Add(A=NamedRerouteUsage(), B=NamedRerouteUsage())
  n115 = Add(A=Multiply(A=Multiply(A=@n108, B=@n111), B=LinearInterpolate(A=@n112, B=@n113, Alpha=@n114)), B=NamedRerouteUsage())
  n116 = Multiply(A=SmoothStep(Value=Add(A=@n107, B=@n115)) {constMax=0.7}) {constB=0.9}
  n117 = NamedRerouteUsage()
  n118 = Saturate(None=Sign(None=NamedRerouteUsage()))
  NamedRerouteDeclaration(None=Multiply(A=Multiply(A=@n106, B=@n116), B=LinearInterpolate(B=@n117, Alpha=@n118) {constA=1})) {name=OpacityOUT, nodeColor=<rgba 0 1 0.5781 1>}
  n119 = TextureSampleParameter2D() {parameterName=StaticOutline, samplerType=SAMPLERTYPE_Masks, texture=/MarksmanVisorGameplay/Assets/Texture/T_VisorMaskTest01}
  NamedRerouteDeclaration(None=@n119.R) {name=Outline, nodeColor=<rgba 1 0 0.6797 1>}
  n120 = NamedRerouteUsage()
  NamedRerouteDeclaration(None=Multiply(
    A=MaterialFunctionCall(uv=@n120) {materialFunction=&MF_UVMask0to1}.Result
    B=TextureSample(UVs=@n120, Tex=@n104) {samplerSource=SSM_Wrap_WorldGroupSettings}.B
   )) {name=FocusFrame, nodeColor=<rgba 0 1 0.8359 1>}
  NamedRerouteDeclaration(None=@n119.G) {name=OutlineSub, nodeColor=<rgba 0.8047 1 0 1>}
  n121 = TextureCoordinate()
  n122 = ComponentMask(None=@n121) {r=true}
  n123 = Constant() {r=0.5}
  NamedRerouteDeclaration(None=SphereMask(
    A=@n122
    B=@n123
    Radius=ScalarParameter() {defaultValue=0.5, group=TechArt_Screen, parameterName=GradientRadius}
    Hardness=ScalarParameter() {group=TechArt_Screen, parameterName=GradientHardness}
   )) {name=GradientU, nodeColor=<rgba 1 0.7734 0 1>}
  NamedRerouteDeclaration(None=@n71) {name=Screen01, nodeColor=<rgba 0.7813 0 1 1>}
  n124 = Subtract(A=Time(), B=ScalarParameter() {group=TechArt_Reactivity, parameterName=ActivateTime})
  n125 = Divide(A=@n124, B=ScalarParameter() {defaultValue=0.5, group=TechArt_Reactivity, parameterName=ActivateTransTime})
  n126 = Saturate(None=Subtract(A=@n124, B=ScalarParameter() {defaultValue=5, group=TechArt_Reactivity, parameterName=ActivateTransDuration}))
  NamedRerouteDeclaration(None=Multiply(A=Saturate(None=@n125), B=OneMinus(None=@n126))) {name=ActivateAlpha, nodeColor=<rgba 1 0.5625 0 1>}
  NamedRerouteDeclaration(None=@n119.B) {name=OutlineGradient, nodeColor=<rgba 0.4531 0 1 1>}
  NamedRerouteDeclaration(None=@n79) {name=CenterUV, nodeColor=<rgba 1 0 0.1641 1>}
  n127 = VectorParameter() {group=UV, parameterName=Bar_CounterUV01, channelNames {a=ScaleY, b=ScaleX, g=CenterY, r=CenterX}, defaultValue=<rgba 0.5 0.5 1 1>}
  MaterialExpressionAppendVector_0 = AppendVector(A=@n127.CenterX, B=@n127.CenterX)
  MaterialExpressionAppendVector_1 = AppendVector(A=@n127.ScaleX, B=@n127.ScaleX)
  MaterialExpressionTextureCoordinate_2 = TextureCoordinate()
  n128 = Custom(UVPoint=@MaterialExpressionAppendVector_0, UVScale=@MaterialExpressionAppendVector_1, OriginalUV=@MaterialExpressionTextureCoordinate_2) {
   code="float2 DistUV =OriginalUV-UVPoint;\r\nDistUV = DistUV/UVScale + float2(0.5,0.5);\r\n\r\nreturn DistUV;"
   description=CopyToPointUV
   outputType=CMOT_Float2
   inputs [
    {inputName=UVPoint, input {expression=&self:MaterialExpressionAppendVector_0}}
    {inputName=UVScale, input {expression=&self:MaterialExpressionAppendVector_1}}
    {inputName=OriginalUV, input {expression=&self:MaterialExpressionTextureCoordinate_2}}
   ]
  }
  n129 = NamedRerouteDeclaration(None=@n128) {name=BarUV, nodeColor=<rgba 1 0 0.1641 1>}
  n130 = NamedRerouteUsage()
  NamedRerouteDeclaration(None=Multiply(
    A=MaterialFunctionCall(uv=@n129) {materialFunction=&MF_UVMask0to1}.Result
    B=TextureSample(UVs=@n129, Tex=@n130) {samplerSource=SSM_Wrap_WorldGroupSettings}.G
   )) {name=Bar, nodeColor=<rgba 1 0 0.5156 1>}
  n131 = Sine(None=Multiply(A=Time(), B=ScalarParameter() {defaultValue=0.2, parameterName=DampingSpeed}))
  n132 = Rotator(
   Coordinate=@n129
   Time=LinearInterpolate(
    A=ScalarParameter() {defaultValue=-1, parameterName=MinAngle}
    B=ScalarParameter() {defaultValue=1, parameterName=MaxAngle}
    Alpha=ConstantBiasScale(None=@n131)
   )
  ) {speed=1}
  NamedRerouteDeclaration(None=Multiply(
    A=MaterialFunctionCall(uv=@n132) {materialFunction=&MF_UVMask0to1}.Result
    B=TextureSample(UVs=@n132, Tex=@n130) {samplerSource=SSM_Wrap_WorldGroupSettings}.R
   )) {name=Bar_Icon, nodeColor=<rgba 0 0.0625 1 1>}
  n133 = VertexColor()
  NamedRerouteDeclaration(None=@n133.R) {name=Vtx_Red, nodeColor=<rgba 0 1 0.5547 1>}
  n134 = VertexNormalWS()
  n135 = Multiply(A=Time(), B=ScalarParameter() {defaultValue=0.2, group=Vtx_Animation, parameterName=IdleSpeed})
  n136 = ConstantBiasScale(None=Sine(None=@n135))
  n137 = Multiply(
   A=NamedRerouteUsage()
   B=Multiply(A=Multiply(A=@n134, B=@n136), B=ScalarParameter() {defaultValue=10, group=Vtx_Animation, parameterName=IdlePieceUnit})
  )
  n138 = Constant3Vector() {constant=<rgba -1 0 0 0>}
  n139 = Constant3Vector() {constant=<rgba 1 0 0 0>}
  n140 = NamedRerouteUsage()
  n141 = Constant3Vector() {desc=There is a rotation in animation BP that make Z the x axis, constant=<rgba 0 0 -1 0>}
  n142 = Constant3Vector() {constant=<rgba 0 0 1 0>}
  n143 = NamedRerouteUsage()
  n144 = StaticSwitchParameter(
   True=Transform(None=LinearInterpolate(A=@n138, B=@n139, Alpha=@n140)) {transformSourceType=TRANSFORMSOURCE_Local}
   False=Transform(None=LinearInterpolate(A=@n141, B=@n142, Alpha=@n143)) {transformSourceType=TRANSFORMSOURCE_Local}
  ) {group=Pickup, parameterName=AsPickupMesh}
  n145 = Subtract(A=@n136) {constB=0.5}
  n146 = Multiply(
   A=Add(A=NamedRerouteUsage(), B=NamedRerouteUsage())
   B=Multiply(A=Multiply(A=@n144, B=@n145), B=ScalarParameter() {defaultValue=4, group=Vtx_Animation, parameterName=IdlePieceUnit_Sub})
  )
  n147 = NamedRerouteUsage()
  n148 = NamedRerouteUsage()
  NamedRerouteDeclaration(None=Add(
    A=NamedRerouteDeclaration(None=@n137) {name=IdleMain, nodeColor=<rgba 1 0 0.6797 1>}
    B=LinearInterpolate(A=@n146, B=@n147, Alpha=@n148) {constB=0}
   )) {name=VtxAnimationOUT, nodeColor=<rgba 0.1484 0 1 1>}
  n149 = Step(Y=@n45) {constX=0.25}
  n150 = Multiply(A=OneMinus(None=@n149), B=@n46)
  n151 = Step(Y=@n45) {constX=0.75}
  n152 = Multiply(A=OneMinus(None=@n49), B=@n151)
  n153 = NamedRerouteUsage()
  NamedRerouteDeclaration(None=Multiply(A=Add(A=@n150, B=@n152), B=Sign(None=@n153))) {name=VtxGroup01, nodeColor=<rgba 1 0 0.6094 1>}
  NamedRerouteDeclaration(None=@n73) {name=Screen02, nodeColor=<rgba 0 1 0.4609 1>}
  n154 = Multiply(A=OneMinus(None=Step(Y=@n45) {constX=0.15}), B=@n149)
  n155 = Multiply(A=OneMinus(None=@n47), B=@n48)
  n156 = OneMinus(None=@n151)
  n157 = Step(Y=@n45) {constX=0.85}
  NamedRerouteDeclaration(None=Add(A=Add(A=@n154, B=@n155), B=Multiply(A=@n156, B=@n157))) {name=VtxGroup02, nodeColor=<rgba 1 0 0.6094 1>}
  NamedRerouteDeclaration(None=@n133.R) {name=Vtx_R, nodeColor=<rgba 1 0 0.7031 1>}
  n158 = Constant3Vector() {constant=<rgba -1 0 -1 0>}
  n159 = NamedRerouteUsage()
  n160 = Step(Y=@n159) {constX=0.65}
  n161 = NamedRerouteUsage()
  n162 = NamedRerouteDeclaration(None=Multiply(A=@n160, B=@n161)) {name=Downpart01, nodeColor=<rgba 1 0 0.3516 1>}
  n163 = Constant3Vector() {constant=<rgba -1 0 0 0>}
  n164 = OneMinus(None=@n160)
  n165 = Step(Y=@n159) {constX=0.85}
  n166 = NamedRerouteDeclaration(None=Multiply(A=@n161, B=Multiply(A=@n164, B=@n165))) {name=Middlepart01, nodeColor=<rgba 0 0.6719 1 1>}
  n167 = Multiply(A=@n161, B=OneMinus(None=@n165))
  n168 = Add(
   A=Add(A=Multiply(A=@n158, B=@n162), B=Multiply(A=@n163, B=@n166))
   B=Multiply(A=Constant3Vector() {constant=<rgba -1 0 1 0>}, B=NamedRerouteDeclaration(None=@n167) {name=Upperpart01, nodeColor=<rgba 0 0.6719 1 1>})
  )
  n169 = NamedRerouteUsage()
  n170 = Step(Y=@n169) {constX=0.65}
  n171 = NamedRerouteUsage()
  n172 = Multiply(
   A=Constant3Vector() {constant=<rgba 1 0 -1 0>}
   B=NamedRerouteDeclaration(None=Multiply(A=@n170, B=@n171)) {name=Downpart02, nodeColor=<rgba 1 0 0.3516 1>}
  )
  n173 = Constant3Vector() {constant=<rgba 1 0 0 0>}
  n174 = OneMinus(None=@n170)
  n175 = Step(Y=@n169) {constX=0.85}
  n176 = NamedRerouteDeclaration(None=Multiply(A=@n171, B=Multiply(A=@n174, B=@n175))) {name=Middlepart02, nodeColor=<rgba 0 0.6719 1 1>}
  n177 = Multiply(A=@n171, B=OneMinus(None=@n175))
  MaterialExpressionAdd_26 = Add(
   A=Add(A=Add(A=@n168, B=@n172), B=Multiply(A=@n173, B=@n176))
   B=Multiply(A=Constant3Vector() {constant=<rgba 1 0 1 0>}, B=NamedRerouteDeclaration(None=@n177) {name=Upperpart02, nodeColor=<rgba 0 0.6719 1 1>})
  )
  n178 = Convert(None=@MaterialExpressionAdd_26) {
   nodeName=Break Float3
   convertInputs [{
     type=Vector3
     expressionInput {
      expression=&self:MaterialExpressionAdd_26
      inputName=None
      mask=0
      maskA=0
      maskB=0
      maskG=0
      maskR=0
      outputIndex=0
     }
    }]
   convertOutputs [{type=Vector3}]
  }
  n179 = Transform(None=@n178) {desc="There is rotation in AnimBP,", transformSourceType=TRANSFORMSOURCE_Local}
  NamedRerouteDeclaration(None=Multiply(A=@n179, B=ScalarParameter() {defaultValue=10, group=Vtx_Animation, parameterName=OffsetUnit_Activate})) {name=ActivateWPO, nodeColor=<rgba 0 0.2734 1 1>}
  n180 = TextureCoordinate()
  n181 = ScalarParameter() {defaultValue=2, parameterName=FlowTiling}
  NamedRerouteDeclaration(None=TextureSample(UVs=Multiply(A=@n180, B=@n181), Tex=@n130) {samplerSource=SSM_Wrap_WorldGroupSettings}.B) {name=flowingPattern, nodeColor=<rgba 0 1 0.4141 1>}
  n182 = Multiply(A=ScalarParameter() {defaultValue=1, group=TechArt_Screen, parameterName=GradientRadius_Depoly}, B=NamedRerouteUsage())
  NamedRerouteDeclaration(None=Saturate(None=SphereMask(A=@n122, B=@n123, Radius=@n182) {hardnessPercent=0})) {name=DepolyU, nodeColor=<rgba 1 0 0.7734 1>}
  NamedRerouteDeclaration(None=@n74) {name=Screen03, nodeColor=<rgba 1 0.1406 0 1>}
  n183 = ComponentMask(None=VectorParameter() {group=TechArt_Reactivity, parameterName=CenterScreen, defaultValue=<rgba 0.5 0.5 0 1>}.RGB) {g=true, r=true}
  n184 = ScalarParameter() {defaultValue=0.4, group=TechArt_Reactivity, parameterName=CenterRadius}
  n185 = NamedRerouteUsage()
  n186 = SphereMask(A=@n121, B=@n183, Radius=Saturate(None=LinearInterpolate(B=@n184, Alpha=@n185) {constA=0.6})) {hardnessPercent=0.2}
  n187 = NamedRerouteUsage()
  NamedRerouteDeclaration(None=Saturate(None=LinearInterpolate(A=@n186, Alpha=@n187))) {name=TargetingMask, nodeColor=<rgba 1 0 0.2344 1>}
  NamedRerouteDeclaration(None=@n154) {name=VtxGroupCenter, nodeColor=<rgba 1 0 0.4687 1>}
  NamedRerouteDeclaration(None=@n100) {name=Depoly, nodeColor=<rgba 0 0.4609 1 1>}
  NamedRerouteDeclaration(None=Multiply(A=@n179, B=ScalarParameter() {defaultValue=-3, group=Vtx_Animation, parameterName=OffsetUnit_Targeting})) {name=TargetingWPO, nodeColor=<rgba 1 0.1641 0 1>}
  n188 = Constant()
  n189 = NamedRerouteUsage()
  n190 = Add(A=NamedRerouteUsage(), B=NamedRerouteUsage())
  NamedRerouteDeclaration(None=Multiply(
    A=ScalarParameter() {defaultValue=1, group=TechArt_Reactivity, parameterName=ScaleSize}
    B=LinearInterpolate(A=@n188, B=@n189, Alpha=@n190) {constB=0}
   )) {name=VtxHideComplex, nodeColor=<rgba 0 0.6719 1 1>}
  NamedRerouteDeclaration(None=@n190) {name=FourSidesPart, nodeColor=<rgba 0 1 0.2734 1>}
  NamedRerouteDeclaration(None=Subtract(A=TransformPosition(None=@n19) {transformType=TRANSFORMPOSSOURCE_World}, B=@n3.XYZ)) {name=LocalScaleTowards, nodeColor=<rgba 0 0.1562 1 1>}
  MaterialExpressionVectorParameter_24 = VectorParameter() {group=TechArt_Reactivity, parameterName=LocalOffset, defaultValue=<rgba 0 0 0 1>}
  n191 = Convert(None=@MaterialExpressionVectorParameter_24.RGB) {
   nodeName=Break Float3
   convertInputs [{
     type=Vector3
     expressionInput {
      expression=&self:MaterialExpressionVectorParameter_24
      inputName=None
      mask=1
      maskA=0
      maskB=1
      maskG=1
      maskR=1
      outputIndex=0
     }
    }]
   convertOutputs [{type=Vector3}]
  }
  n192 = VectorParameter() {group=TechArt_Reactivity, parameterName=CenterOffsetWS, defaultValue=<rgba 0 0 0 1>}
  NamedRerouteDeclaration(None=Add(A=Transform(None=@n191) {transformSourceType=TRANSFORMSOURCE_Local}, B=ComponentMask(None=@n192.RGB) {b=true, g=true, r=true})) {name=CenterWS, nodeColor=<rgba 0 1 0.0391 1>}
  NamedRerouteDeclaration(None=LocalPosition() {localOrigin=InstancePreSkinning}.XYZ) {name=Preskinned, nodeColor=<rgba 1 0.3281 0 1>}
  NamedRerouteDeclaration(None=Sign(None=NamedRerouteDeclaration(None=@n133.G) {name=Vtx_G, nodeColor=<rgba 0.1139 1 0 1>})) {name=VtxGroup03, nodeColor=<rgba 0 0.0859 1 1>}
  NamedRerouteDeclaration(None=VectorParameter() {group=TechArt_Reactivity, parameterName=LocalOffset, defaultValue=<rgba 0 0 0 1>}.RGB) {name=LocalBone, nodeColor=<rgba 0 0.6016 1 1>}
  NamedRerouteDeclaration(None=StaticSwitchParameter(True=NamedRerouteUsage(), False=NamedRerouteUsage()) {defaultValue=true, parameterName=UseTargetingAsActiveForColor}) {name=ColorReactiveAlpha, nodeColor=<rgba 1 0.0703 0 1>}
  n193 = SphereMask(
   A=SphereMask(A=@n121, B=@n183, Radius=Constant() {r=0.5}) {hardnessPercent=0}
   B=Power(Base=NamedRerouteUsage()) {constExponent=0.5}
   Radius=Constant() {r=0.1}
  ) {hardnessPercent=0.5}
  n194 = NamedRerouteUsage()
  NamedRerouteDeclaration(None=Multiply(A=LinearInterpolate(A=@n193, Alpha=@n194) {constB=0}, B=@n193)) {name=TargetingRing, nodeColor=<rgba 0.2422 0 1 1>}
  NamedRerouteDeclaration(None=Sign(None=NamedRerouteDeclaration(None=@n133.B) {name=Vtx_B, nodeColor=<rgba 0 0.7422 1 1>})) {name=VtxGroup04, nodeColor=<rgba 0.4062 1 0 1>}
  NamedRerouteUsage()
  NamedRerouteUsage()
  NamedRerouteUsage()
  NamedRerouteUsage()
  NamedRerouteUsage()
  NamedRerouteUsage()
  NamedRerouteUsage()
  NamedRerouteUsage()
  ObjectPositionWS()
  OneMinus(None=NamedRerouteUsage())
  Panner(Coordinate=@n129) {speedX=0.2}
  PreSkinnedPosition()
  Sine(None=@n135)
  StaticSwitchParameter(True=ScalarParameter() {defaultValue=0.5, parameterName=VisorOpacity}, False=NamedRerouteUsage()) {group=Vis, parameterName="[Debug]Pickup Mesh"}
  n195 = NamedRerouteDeclaration(None=Saturate(None=ScalarParameter() {group=TechArt_Reactivity, parameterName=IsTargeting})) {name=TargetingAlpha, nodeColor=<rgba 0 1 0.0859 1>}
  n196 = Sign(None=@n195)
  n197 = NamedRerouteUsage()
  n198 = LinearInterpolate(
   A=NamedRerouteUsage()
   B=Multiply(A=NamedRerouteUsage(), B=Power(Base=@n195) {constExponent=3})
   Alpha=Multiply(A=Saturate(None=@n196), B=OneMinus(None=@n197))
  )
  n199 = Subtract(A=ObjectPositionWS(), B=WorldPosition().XYZ)
  n200 = Add(A=NamedRerouteUsage(), B=NamedRerouteUsage())
  StaticSwitchParameter(
   True=Constant()
   False=StaticSwitchParameter(True=LinearInterpolate(A=@n198, B=@n199, Alpha=@n200), False=@n198) {group=Vis, parameterName="[Debug]HideWings"}
  ) {group=Vis, parameterName="[Debug]Pickup Mesh"}
  Subtract(A=Constant(), B=WorldPosition().XYZ)
  Subtract(A=@MaterialExpressionTransformPosition_1, B=@n20)
 ]
}
