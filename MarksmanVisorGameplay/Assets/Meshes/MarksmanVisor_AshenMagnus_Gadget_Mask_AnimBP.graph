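# Animation Blueprint for the Ashen Magnus Marksman Visor mask gadget, positioning it as a character customization part.
# From a cached DefaultSlot pose, it pitches root -90 degrees and adds LocalOffsetT/R to the offset bone.
# bSpineSocket blends between two such branches differing in translation space (parent-bone or bone), followed by a fixed offset nudge.
# The event graph binds customization-finished to the mask base class's native handler and unbinds it on PreDestruction.
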AnimBlueprint /MarksmanVisorGameplay/Assets/Meshes/MarksmanVisor_AshenMagnus_Gadget_Mask_AnimBP
target_skeleton=&/MarksmanVisor_AshenMagnus_Gadget_Skeleton
graphs [
 AnimGraph {nodes [
   n1 = "Transform (Modify) Bone - Bone: root"(ComponentPose=Local To Component(LocalPose=Use cached pose 'StartPose'()), Translation="0,0,0", Rotation="-90,0,0") {
    rotationMode=BMM_Additive
    rotationSpace=BCS_BoneSpace
    alphaBoolBlend {blendOption=Linear}
    boneToModify {boneName=root}
    rotation {pitch=-90}
   }
   n2 = "Transform (Modify) Bone - Bone: offset"(
    ComponentPose=@n1
    Alpha=1
    Translation=Get LocalOffsetT() -> (LocalOffsetT:Vector)
    Rotation=Get LocalOffsetR() -> (LocalOffsetR:Rotator)
    Scale="1,1,1"
   ) {
    rotationMode=BMM_Additive
    rotationSpace=BCS_ParentBoneSpace
    scaleMode=BMM_Additive
    scaleSpace=BCS_ParentBoneSpace
    translationMode=BMM_Additive
    translationSpace=BCS_BoneSpace
    alphaBoolBlend {blendOption=Linear}
    boneToModify {boneName=offset}
   }
   n3 = Component To Local(ComponentPose=@n2)
   n4 = "Transform (Modify) Bone - Bone: root"(ComponentPose=Local To Component(LocalPose=Use cached pose 'StartPose'()), Translation="0,0,0", Rotation="-90,0,0") {
    rotationMode=BMM_Additive
    rotationSpace=BCS_BoneSpace
    alphaBoolBlend {blendOption=Linear}
    boneToModify {boneName=root}
    rotation {pitch=-90}
   }
   n5 = "Transform (Modify) Bone - Bone: offset"(ComponentPose=@n4, Alpha=1, Translation=Get LocalOffsetT(), Rotation=Get LocalOffsetR(), Scale="1,1,1") {
    rotationMode=BMM_Additive
    rotationSpace=BCS_ParentBoneSpace
    scaleMode=BMM_Additive
    scaleSpace=BCS_ParentBoneSpace
    translationMode=BMM_Additive
    translationSpace=BCS_ParentBoneSpace
    alphaBoolBlend {blendOption=Linear}
    boneToModify {boneName=offset}
   }
   n6 = Component To Local(ComponentPose=@n5)
   n7 = Get bSpineSocket() -> (bSpineSocket:bool)
   n8 = "Transform (Modify) Bone - Bone: offset"(
    ComponentPose=Local To Component(LocalPose=Blend Poses by bool(BlendPose_0=@n6, BlendPose_1=@n3, bActiveValue=@n7) {blendType=Linear})
    Alpha=1
    Translation="0,3,7"
    Rotation="0,0,0"
    Scale="1,1,1"
   ) {
    rotationMode=BMM_Additive
    rotationSpace=BCS_ParentBoneSpace
    scaleMode=BMM_Additive
    scaleSpace=BCS_ParentBoneSpace
    translationMode=BMM_Additive
    translationSpace=BCS_ParentBoneSpace
    alphaBoolBlend {blendOption=Linear}
    boneToModify {boneName=offset}
    translation {y=3, z=7}
   }
   n9 = "Transform (Modify) Bone - Bone: main"(ComponentPose=@n8, Alpha=1, Translation="0,0,0", Rotation="0,0,0", Scale="1,1,1") {translationMode=BMM_Additive, translationSpace=BCS_BoneSpace, alphaBoolBlend {blendOption=Linear}, boneToModify {boneName=main}}
   Output Pose(Result=Component To Local(ComponentPose=@n9))
   Save cached pose 'StartPose'(Pose=Slot 'DefaultSlot'()) {cacheName=StartPose}
   Get LocalOffsetS() -> (LocalOffsetS:Vector)
   Get Floater() -> (Floater:Vector)
   Get LocalOffsetS()
  ]}
 EventGraph {nodes [
   n1 = CharacterCustomizationFinished_Event() {is=title K2Node_CustomEvent}
   > NativeCharacterCustomizationFinished(self:Mask Base CCPAnim Instance Object Reference)
   Event BlueprintBeginPlay() {is=title K2Node_Event}
   > Parent: BlueprintBeginPlay() {is=title K2Node_CallParentFunction}
   > Assign Character Customization Finished!(execute:Exec, self:Custom Character Part Anim Instance Object Reference, Delegate:"Delegate (by ref)"=@n1) -> (then:Exec) {is=title K2Node_AssignDelegate}
   Event PreDestruction() {is=title K2Node_Event}
   > Parent: PreDestruction() {is=title K2Node_CallParentFunction}
   > Unbind Event from Character Customization Finished(self:Custom Character Part Anim Instance Object Reference, Delegate=@n1) {is=title K2Node_RemoveDelegate}
  ]}
]
variables []
